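; Base Blueprint for Athena SkyTubes, spline-based wind tunnels that carry players, pawns, projectiles and vehicles.
; The construction script builds smoothed, scaled spline meshes along the spline, retaining a deprecated build path.
; Overlaps route actors by type into enter/exit calls, managing skydiving, gameplay effects, quest credit, vehicle knockback, sounds and a repeating destruction-effect timer.
; Enable and shutdown events toggle audio and visibility, while CalculateLaunchForce scales boost by view-to-spline angle.
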(event EventOnStartShuttingDown
  (bind _fortsplineaudio (Variables|Default|GetFortSplineAudio))
  (bind _ground_particlesystem (Variables|Default|GetGround_ParticleSystem))
  (|CallParentFunction)
  (Variables|/Script/AudioShapes.AudioShapeComponent|SetFadeOutTime (SkyTube|GetDelayUntilDisabled) _fortsplineaudio)
  (Shape|Disable _fortsplineaudio)
  (bind _returnvalue (Control|GetAudioComponent _fortsplineaudio))
  (Utilities|IsValid _returnvalue
    (:Is%20Valid
      (Audio|PlaySoundatLocation "/Game/Athena/Sounds/Ambient/Skytube/Skytube_Disable_Cue.Skytube_Disable_Cue" (Transformation|GetWorldLocation _returnvalue)))
    (:Is%20Not%20Valid
      _))
  (Utilities|IsValid _ground_particlesystem
    (:Is%20Valid
      (Rendering|SetActorHiddenInGame _ground_particlesystem true))))

(event EventBeginPlay
  (if (not (SkyTube|IsSkyTubeEnabled))
    (Shape|Disable (Variables|Default|GetFortSplineAudio))
    (Rendering|SetActorHiddenInGame true)))

(event EventOnEnabledChanged
  (bind _fortsplineaudio (Variables|Default|GetFortSplineAudio))
  (bind _ground_particlesystem (Variables|Default|GetGround_ParticleSystem))
  (if (SkyTube|IsSkyTubeEnabled)
    (Shape|Enable _fortsplineaudio)
    (Rendering|SetActorHiddenInGame false)
    (Utilities|IsValid _ground_particlesystem
      (:Is%20Valid
        (Rendering|SetActorHiddenInGame _ground_particlesystem false)))
    (else
      (Shape|Disable _fortsplineaudio)
      (Rendering|SetActorHiddenInGame true)
      (Utilities|IsValid _ground_particlesystem
        (:Is%20Valid
          (Rendering|SetActorHiddenInGame _ground_particlesystem true))))))

(event Custom|PlaySkytubeExitSound
  ;[0.3991 0.436 1] Player Enter/Exit Sounds
  (bind _asfort_player_pawn (Utilities|Casting|CastToFortPlayerPawn (stale Collision|EventActorBeginOverlap OtherActor)))
  (bind _player_exit_tube_sound (Variables|Audio|GetPlayer%20Exit%20Tube%20Sound)
    (:then
      (Audio|SpawnSoundAttached _player_exit_tube_sound (Variables|/Script/Engine.Character|GetMesh _asfort_player_pawn)))
    (:else)))

(event Custom|PlaySkytubeEnterSound
  ;[0.3991 0.436 1] Player Enter/Exit Sounds
  (bind _asfort_player_pawn (Utilities|Casting|CastToFortPlayerPawn (stale Collision|EventActorBeginOverlap OtherActor)))
  (Audio|SpawnSoundAttached (Variables|Audio|GetPlayer%20Enter%20Tube%20Sound) (Variables|/Script/Engine.Character|GetMesh _asfort_player_pawn)))

(event Collision|EventActorBeginOverlap
  (params (OtherActor :type Actor%20Object%20Reference))
  (bind _returnvalue (Actor|GetComponentbyClass OtherActor "/Script/FortniteGame.FortSkyTubePhysicsComponent"))
  (if (SkyTube|IsSkyTubeEnabled)
    (bind _outblocked (CallFunction|CheckforBlocked OtherActor))
    (if _outblocked
      _
      (elif (Utilities|Time|IsValidTimerHandle (Variables|Default|GetSkyTubeDestructionGEApplication_Timer))
        _
        (else
          (bind _returnvalue_2 (Utilities|Time|SetTimerbyEvent (EventDispatchers|CreateEvent :!event Custom|UpdateSkyTubeDestructionGE) 1.0 true))
          (Variables|Default|SetSkyTubeDestructionGEApplication_Timer _returnvalue_2))))
    (bind _asfort_projectile_base (Utilities|Casting|CastToFortProjectileBase OtherActor)
      (:then
        (Projectile|EnterSkyTube _asfort_projectile_base self)
        (CallFunction|PlaySkytubeEnterSound))
      (:CastFailed
        (bind _asfort_player_pawn (Utilities|Casting|CastToFortPlayerPawn OtherActor)
          (:then
            (bind _boutisfirsttube (SkyTube|EnterSkyTube _asfort_player_pawn self :!class /Script/FortniteGame.FortPlayerPawn))
            ;[1 0.348 0] This is added to handle the strange issue where Witches Brooms would not correctly enter skydive - players would abe able to shoot & aim whilst in the tube. Setting the falling movement mode fixes this. - Mike Jones
            (Utilities|FlowControl|Sequence
              (:then_0
                (CallFunction|PlaySkytubeEnterSound))
              (:then_1
                (if (GameplayTags|HasMatchingGameplayTag _asfort_player_pawn "(TagName=\"Granted.Athena.Power.LeapGlider.Generic.Active\")")
                  (Pawn|Components|CharacterMovement|SetMovementMode (Variables|/Script/Engine.Character|GetCharacterMovement _asfort_player_pawn) "MOVE_Falling" 0))
                (Mission|Skydiving|Skydive false _asfort_player_pawn) ; If player is using glider, put in skydive state
                (if _boutisfirsttube
                  (GameplayEffects|ApplyGameplayEffectToTarget (Ability|GetAbilitySystemComponent _asfort_player_pawn) "/Game/Athena/Environments/Blueprints/SkyTube/GE_Athena_SkyTube_Loop.GE_Athena_SkyTube_Loop_C" (Ability|GetAbilitySystemComponent _asfort_player_pawn) 0.0 (Utilities|Struct|MakeGameplayEffectContextHandle))
                  (if (Networking|IsServer)
                    (Quest|SendComplexCustomStatEvent (Quests|GetQuestManager (Fortnite|GetFortPlayerControllerfromActor _asfort_player_pawn) "Invalid") 0 "(GameplayTags=((TagName=\"Athena.Quests.WindTunnel\")))" "(GameplayTags=)" 1))))))
          (:CastFailed
            (bind _asfort_pawn (Utilities|Casting|CastToFortPawn OtherActor)
              (:then
                (SkyTube|EnterSkyTube _asfort_pawn self :!class /Script/FortniteGame.FortPawn)
                (CallFunction|PlaySkytubeEnterSound))
              (:CastFailed
                (if (Class|ItemUIFunctionLibrary|IsNullItem _returnvalue)
                  (bind _asfort_athena_vehicle (Utilities|Casting|CastToFortAthenaVehicle OtherActor))
                  (CallFunction|LaunchVehicle _asfort_athena_vehicle)
                  ;[1 1 1] [Old] Looping GE Applied to vehicles on entrance
                  (CallFunction|PlaySkytubeVehEnterSound)
                  (else
                    (SkyTube|EnterSkyTube _returnvalue self :!class /Script/FortniteGame.FortSkyTubePhysicsComponent)
                    (Utilities|FlowControl|Sequence
                      (:then_0
                        (CallFunction|PlaySkytubeEnterSound))
                      (:then_1
                        (bind _asfort_vehicle_interface (Utilities|Casting|CastToFortVehicleInterface OtherActor))
                        (Environment|ApplyGlobalEnvironmentGameplayEffecttoActor (Vehicle|GetActor _asfort_vehicle_interface) "/Game/Athena/Environments/Blueprints/SkyTube/GE_Athena_SkyTube_Loop.GE_Athena_SkyTube_Loop_C")))))))))))))

(event Custom|PlaySkytubeVehEnterSound
  ;[0.3991 0.436 1] Vehicle Enter/Exit Sounds
  (bind _asfort_athena_vehicle (Utilities|Casting|CastToFortAthenaVehicle (stale Collision|EventActorBeginOverlap OtherActor)))
  (Audio|SpawnSoundAttached (Variables|Audio|GetVeh%20Enter%20Tube%20Sound) (Actor|GetParentComponent _asfort_athena_vehicle)))

(event Custom|PlaySkytubeVehExitSound
  ;[0.3991 0.436 1] Vehicle Enter/Exit Sounds
  (bind _asfort_athena_vehicle (Utilities|Casting|CastToFortAthenaVehicle (stale Collision|EventActorBeginOverlap OtherActor)))
  (Audio|SpawnSoundAttached (Variables|Audio|GetVeh%20Exit%20Tube%20Sound) (Actor|GetParentComponent _asfort_athena_vehicle)))

(event Collision|EventActorEndOverlap
  (params (OtherActor :type Actor%20Object%20Reference))
  (CallFunction|ActorExitSkytube OtherActor)
  (CallFunction|PlaySkytubeExitSound))

(event SkyTube|EventHandleForcedExit
  (params (ExitingPawn :type Actor%20Object%20Reference))
  (CallFunction|ActorExitSkytube ExitingPawn)
  (CallFunction|PlaySkytubeExitSound))

(event Custom|UpdateSkyTubeDestructionGE
  (bind _overlappingactors (Collision|GetOverlappingActors "/Script/FortniteGame.FortPlayerPawn"))
  (if (> (Utilities|Array|Length _overlappingactors) 0)
    (for _array_element _overlappingactors
      (bind _asfort_player_pawn (Utilities|Casting|CastToFortPlayerPawn _array_element))
      (GameplayEffects|ApplyGameplayEffectToTarget (Ability|GetAbilitySystemComponent _asfort_player_pawn) "/Game/Athena/Environments/Blueprints/SkyTube/GE_Athena_SkyTube_Destruction.GE_Athena_SkyTube_Destruction_C" (Ability|GetAbilitySystemComponent _asfort_player_pawn) 0.0 (Utilities|Struct|MakeGameplayEffectContextHandle)))
    (else
      (Utilities|Time|ClearandInvalidateTimerbyHandle (Variables|Default|GetSkyTubeDestructionGEApplication_Timer)))))

(fn ActorExitSkytube
  (params (OtherActor :type Actor%20Object%20Reference))
  (bind _returnvalue (Actor|GetComponentbyClass OtherActor "/Script/FortniteGame.FortSkyTubePhysicsComponent"))
  (Utilities|FlowControl|Sequence
    (:then_0
      (bind _asfort_projectile_base (Utilities|Casting|CastToFortProjectileBase OtherActor)
        (:then
          (Projectile|ExitSkyTube _asfort_projectile_base self))
        (:CastFailed
          (bind _asfort_player_pawn (Utilities|Casting|CastToFortPlayerPawn OtherActor)
            (:then
              (bind _boutislasttube (SkyTube|ExitSkyTube _asfort_player_pawn self :!class /Script/FortniteGame.FortPlayerPawn))
              (if _boutislasttube
                (GameplayEffects|RemoveActiveGameplayEffectbySourceEffect (Ability|GetAbilitySystemComponent _asfort_player_pawn) "/Game/Athena/Environments/Blueprints/SkyTube/GE_Athena_SkyTube_Loop.GE_Athena_SkyTube_Loop_C" 0 -1)
                (GameplayEffects|RemoveActiveGameplayEffectbySourceEffect (Ability|GetAbilitySystemComponent _asfort_player_pawn) "/Game/Athena/Environments/Blueprints/SkyTube/GE_Athena_SkyTube_Destruction.GE_Athena_SkyTube_Destruction_C" 0 -1) ; Remove loopng GE from player for Destruction on Exit
                (Variables|/Script/FortniteGame.FortPawn|SetbIgnoreNextFallingDamage true _asfort_player_pawn)))
            (:CastFailed
              (bind _asfort_pawn (Utilities|Casting|CastToFortPawn OtherActor)
                (:then
                  (SkyTube|ExitSkyTube _asfort_pawn self :!class /Script/FortniteGame.FortPawn))
                (:CastFailed
                  (Utilities|IsValid _returnvalue
                    (:Is%20Valid
                      (SkyTube|ExitSkyTube _returnvalue self :!class /Script/FortniteGame.FortSkyTubePhysicsComponent)
                      (bind _asfort_vehicle_interface (Utilities|Casting|CastToFortVehicleInterface OtherActor))
                      (GameplayEffects|RemoveActiveGameplayEffectbySourceEffect (Vehicle|GetVehicleAbilitySystemComponent _asfort_vehicle_interface) "/Game/Athena/Environments/Blueprints/SkyTube/GE_Athena_SkyTube_Loop.GE_Athena_SkyTube_Loop_C" 0 -1) ; Remove loopng GE from vehicles on exit
                    )))))))))
    (:then_1
      (bind _asfort_player_pawn_2 (Utilities|Casting|CastToFortPlayerPawn OtherActor)
        (:then
          (GameplayEffects|ApplyGameplayEffectToSelf :self (Ability|GetAbilitySystemComponent _asfort_player_pawn_2) :GameplayEffectClass "/Game/Athena/Items/QuestInteractables/Quest_GEs/GE_Athena_Quests_PostSlipstreamTag.GE_Athena_Quests_PostSlipStreamTag_C" :Level 0.0))
        (:CastFailed
          (bind _asfort_vehicle_interface_2 (Utilities|Casting|CastToFortVehicleInterface OtherActor))
          (for _array_element (Vehicle|GetAllPassengers _asfort_vehicle_interface_2)
            (GameplayEffects|ApplyGameplayEffectToSelf :self (Ability|GetAbilitySystemComponent _array_element) :GameplayEffectClass "/Game/Athena/Items/QuestInteractables/Quest_GEs/GE_Athena_Quests_PostSlipstreamTag.GE_Athena_Quests_PostSlipStreamTag_C" :Level 0.0)))))))

(fn AddSplineMesh
  (bind _returnvalue (AddComponent|NODE_AddSplineMeshComponent-0))
  (Components|StaticMesh|SetStaticMesh _returnvalue (Variables|Default|GetSplineStaticMesh))
  (Rendering|Material|SetMaterial _returnvalue 0 (Variables|Default|GetSplineMeshMaterialOverride))
  (if (SkyTube|IsExemptfromShutdown)
    (bind _returnvalue_2 (Rendering|Material|CreateDynamicMaterialInstance _returnvalue 0 0 "None" :!class /Script/Engine.PrimitiveComponent))
    (Rendering|Material|SetScalarParameterValue _returnvalue_2 "ExemptFromShutdown" 1.0 :!class /Script/Engine.MaterialInstanceDynamic))
  (SplineMesh|SetForwardAxis _returnvalue (Variables|Default|GetMeshForwardAxis) true)
  (Collision|SetCollisionEnabled _returnvalue (Variables|Default|GetCollisionEnabled))
  (Collision|SetCollisionProfileName _returnvalue (Variables|Default|GetCollisionProfileName) true :!class /Script/Engine.PrimitiveComponent)
  (Variables|/Script/Engine.PrimitiveComponent|SetbGenerateOverlapEvents true _returnvalue)
  (HLOD|SetExcludedFromHLODLevel _returnvalue "HLOD0" (Variables|Default|GetExcludeFromMediumHLOD))
  (return _returnvalue))

(fn CalculateLaunchForce
  (params (Pawn :type Fort%20Player%20Pawn%20Object%20Reference) (Velocity :type Vector))
  (bind _returnvalue (Transformation|GetActorLocation Pawn))
  (bind _spline (Variables|SkyTube|GetSpline))
  (bind _returnvalue_2 (Spline|FindDirectionClosesttoWorldLocation _spline _returnvalue "World"))
  (bind _returnvalue_3 (Pawn|GetControlRotation Pawn :!class /Script/Engine.Pawn))
  (bind _returnvalue_4 (Math|Rotator|GetRotationXVector _returnvalue_3))
  (bind _returnvalue_5 (Math|Vector|Normalize _returnvalue_4))
  (bind _returnvalue_6 (Math|Vector|DotProduct _returnvalue_5 _returnvalue_2))
  (bind _returnvalue_7 (Math|Float|Absolute(Float) _returnvalue_6))
  (bind _returnvalue_8 (Math|Trig|Acos(Degrees) _returnvalue_7))
  (bind (_returnvalue_x _returnvalue_y _returnvalue_z) (Math|Vector|BreakVector (* Velocity (Variables|Default|GetFinalBoostMultiplier))))
  (Variables|Default|SetFinalBoostMultiplier (Math|Float|MapRangeClamped _returnvalue_8 9e+01 0.0 (ScalableFloat|GetValueatLevel (Variables|Default|GetDataMinBoostMultiplier)) (ScalableFloat|GetValueatLevel (Variables|Default|GetDataMaxBoostMultiplier)))) ; Set boost multipliers
  (Variables|Default|SetFinalBoostFlatForce (Math|Float|MapRangeClamped _returnvalue_8 0.0 9e+01 (ScalableFloat|GetValueatLevel (Variables|Default|GetDataMinFlatZBoost)) (ScalableFloat|GetValueatLevel (Variables|Default|GetDataMaxFlatZBoost))))
  (CallFunction|CheckZDirection Pawn Velocity) ; flip z if necessary
  (return (Math|Vector|SelectVector (Math|Vector|vector+vector (Math|Vector|vector+vector (Math|Vector|vector*vector (Math|Vector|Normalize (Pawn|Components|CharacterMovement|GetCurrentAcceleration (Variables|/Script/Engine.Character|GetCharacterMovement Pawn))) (Math|Vector|MakeVector (ScalableFloat|GetValueatLevel (Variables|Default|GetDataHorizontalAccelerationMultiplier)) (ScalableFloat|GetValueatLevel (Variables|Default|GetDataHorizontalAccelerationMultiplier)))) (Math|Vector|MakeVector 0.0 0.0 (Variables|Default|GetFinalBoostFlatForce))) (Math|Vector|MakeVector _returnvalue_x _returnvalue_y (* _returnvalue_z (ScalableFloat|GetValueatLevel (Variables|Default|GetDataVerticalVelocityMultiplier))))) "0, 0, 0" (SkyTube|IsSkyTubeEnabled))))

(fn CheckForBlocked
  (params (InActor :type Actor%20Object%20Reference))
  (return (GameplayTags|HasMatchingGameplayTag (Ability|GetAbilitySystemComponent InActor) (Variables|Default|GetSkytubeBlockedTag))))

(fn CheckZDirection
  (params (Pawn :type Fort%20Player%20Pawn%20Object%20Reference) (Velocity :type Vector))
  (if (not (>= (Math|Vector|DotProduct (Math|Vector|Normalize (Transformation|GetActorUpVector Pawn)) (Math|Vector|Normalize Velocity)) -0.01))
    (Variables|Default|SetFinalBoostFlatForce (* (Variables|Default|GetFinalBoostFlatForce) -1.0)))
  (return))

(fn LaunchVehicle
  (params (Vehicle :type Fort%20Athena%20Vehicle%20Object%20Reference))
  (Fortnite|ApplyGameplayEffectsToActor Vehicle (Utilities|Array|MakeArray "/Game/Athena/Environments/Blueprints/SkyTube/GE_Athena_SkyTube_GrantKnockback.GE_Athena_SkyTube_GrantKnockback_C") (Spline|FindLocationClosesttoWorldLocation (Variables|SkyTube|GetSpline) (Transformation|GetActorLocation Vehicle) "World") false 0 (Variables|Default|GetAbilitySystem)))

(fn SmoothTangents
  (params (Location%20A :type Vector) (Tangent%20A :type Vector) (Location%20B :type Vector) (Tangent%20B :type Vector))
  (return (Math|Vector|SelectVector (* (Math|Vector|Normalize (Variables|GetTangent%20A)) (* (Math|Vector|VectorLength (- (Variables|GetLocation%20A) (Variables|GetLocation%20B))) (Variables|Default|GetTension))) (Variables|GetTangent%20A) (Variables|Default|GetAutoSmoothTangent)) (Math|Vector|SelectVector (* (Math|Vector|Normalize (Variables|GetTangent%20B)) (* (Math|Vector|VectorLength (- (Variables|GetLocation%20A) (Variables|GetLocation%20B))) (Variables|Default|GetTension))) (Variables|GetTangent%20B) (Variables|Default|GetAutoSmoothTangent))))

(fn ConstructionScript
  (bind _spline (Variables|SkyTube|GetSpline))
  (bind _spline_2 (Variables|SkyTube|GetSpline))
  (bind _tubescale (Variables|Default|GetTubeScale))
  (bind _spline_3 (Variables|SkyTube|GetSpline))
  (bind _spline_4 (Variables|SkyTube|GetSpline))
  (|CallParentFunction)
  (if (Variables|Default|GetDEPRECATED_UseOldConstructionScript)
    ;[1.561 0.2791 0] DEPRECATED - Old construction script - I kept it in case we need it for old sky tubes
    (for _index (range (+ (- (Spline|GetNumberOfSplinePoints (Variables|SkyTube|GetSpline)) 2) 1))
      (bind _newsplinemesh (CallFunction|AddSplineMesh))
      (bind _output_get (LocalVariables|SetNewSplineMesh _newsplinemesh))
      (bind (_tangent_0 _tangent_1) (CallFunction|SmoothTangents (Spline|GetLocationandTangentatSplinePoint _spline_4 _index "Local") (.Tangent (Spline|GetLocationandTangentatSplinePoint _spline_4 _index "Local")) (Spline|GetLocationandTangentatSplinePoint _spline_4 (+ _index 1) "Local") (.Tangent (Spline|GetLocationandTangentatSplinePoint _spline_4 (+ _index 1) "Local"))))
      (SplineMesh|SetStartandEnd _output_get (Spline|GetLocationandTangentatSplinePoint _spline_4 _index "Local") _tangent_0 (Spline|GetLocationandTangentatSplinePoint _spline_4 (+ _index 1) "Local") _tangent_1 false :!class /Script/Engine.SplineMeshComponent)
      (Utilities|Array|SetArrayElem (Variables|Default|GetSplinePointTransforms) _index (Math|Transform|MakeTransform (Spline|GetLocationandTangentatSplinePoint (Variables|SkyTube|GetSpline) _index "World") (Math|Conversions|RotationFromXVector (.Tangent (Spline|GetLocationandTangentatSplinePoint (Variables|SkyTube|GetSpline) _index "World")))) true)
      (SplineMesh|SetStartScale (LocalVariables|GetNewSplineMesh) (Math|Vector2D|MakeVector2D (.x (Spline|GetScaleatSplinePoint _spline_3 0)) (.y (Spline|GetScaleatSplinePoint _spline_3 0))) false :!class /Script/Engine.SplineMeshComponent)
      (SplineMesh|SetEndScale (LocalVariables|GetNewSplineMesh) (Math|Vector2D|MakeVector2D (.x (Spline|GetScaleatSplinePoint _spline_3 0)) (.y (Spline|GetScaleatSplinePoint _spline_3 0))) false :!class /Script/Engine.SplineMeshComponent)
      (SplineMesh|UpdateMesh (LocalVariables|GetNewSplineMesh)))
    (elif (Spline|IsClosedLoop _spline)
      (Variables|Default|SetSplinePointSubtractor 1)))
  (for _index_2 (range (+ (- (Spline|GetNumberOfSplinePoints _spline) (Variables|Default|GetSplinePointSubtractor)) 1))
    (bind _newsplinemesh_2 (CallFunction|AddSplineMesh))
    (bind _output_get_2 (LocalVariables|SetNewSplineMesh _newsplinemesh_2))
    (bind (_tangent_0_2 _tangent_1_2) (CallFunction|SmoothTangents (Spline|GetLocationandTangentatSplinePoint _spline_2 _index_2 "Local") (.Tangent (Spline|GetLocationandTangentatSplinePoint _spline_2 _index_2 "Local")) (Spline|GetLocationandTangentatSplinePoint _spline_2 (+ _index_2 1) "Local") (.Tangent (Spline|GetLocationandTangentatSplinePoint _spline_2 (+ _index_2 1) "Local"))))
    (SplineMesh|SetStartandEnd _output_get_2 (Spline|GetLocationandTangentatSplinePoint _spline_2 _index_2 "Local") _tangent_0_2 (Spline|GetLocationandTangentatSplinePoint _spline_2 (+ _index_2 1) "Local") _tangent_1_2 false :!class /Script/Engine.SplineMeshComponent)
    (SplineMesh|SetStartRollRadians _output_get_2 (Math|Trig|DegreesToRadians (.roll (Spline|GetRotationatSplinePoint _spline_2 _index_2 "Local"))) true)
    (SplineMesh|SetEndRollRadians _output_get_2 (Math|Trig|DegreesToRadians (.roll (Spline|GetRotationatSplinePoint _spline_2 (+ _index_2 1) "Local"))) true)
    (SplineMesh|SetStartScale _output_get_2 (* (Math|Vector2D|MakeVector2D (.y (Spline|GetScaleatSplinePoint _spline_2 _index_2)) (.z (Spline|GetScaleatSplinePoint _spline_2 _index_2))) _tubescale) true :!class /Script/Engine.SplineMeshComponent)
    (SplineMesh|SetEndScale _output_get_2 (* (Math|Vector2D|MakeVector2D (.y (Spline|GetScaleatSplinePoint _spline_2 (+ _index_2 1))) (.z (Spline|GetScaleatSplinePoint _spline_2 (+ _index_2 1)))) _tubescale) true :!class /Script/Engine.SplineMeshComponent)
    ;[1 1 1] Expose spline point transforms in details panel for copying and pasting
    (Utilities|Array|SetArrayElem (Variables|Default|GetSplinePointTransforms) _index_2 (Math|Transform|MakeTransform (Spline|GetLocationandTangentatSplinePoint (Variables|SkyTube|GetSpline) _index_2 "World") (Math|Conversions|RotationFromXVector (.Tangent (Spline|GetLocationandTangentatSplinePoint (Variables|SkyTube|GetSpline) _index_2 "World")))) true)))
