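; Cosmetic SkyTube conduit ring prop with ambient audio, a wind tunnel mesh and pass-through feedback.
; On remote instances, BeginPlay binds capsule overlap when enabled, otherwise stops audio and hides the wind mesh; the playlist-loaded handler is empty.
; Overlaps trigger camera shake and a local GameplayCue scaled by pawn speed, while enable changes fade audio.
; Construction selects the wind mesh and either removes the pole or positions and rotates the base.
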(event EventOnEnabledChanged
  (bind _ambientaudio (Variables|Default|GetAmbientAudio))
  (|CallParentFunction)
  (if (SkyTube|IsEnabled)
    (Audio|Components|Audio|FadeIn (Variables|Default|GetAmbientAudio) 1.0 1.0 0.0 "Linear" :!class /Script/Engine.AudioComponent)
    (else
      (Audio|Components|Audio|FadeOut _ambientaudio 1.0 0.0 "Linear" :!class /Script/Engine.AudioComponent)
      (Audio|PlaySoundatLocation "/Game/Athena/Sounds/Ambient/Skytube/SC_Skytube_Disable_S9.SC_Skytube_Disable_S9" (Transformation|GetWorldLocation _ambientaudio)))))

(event EventBeginPlay
  (Networking|SwitchHasAuthority
    (:Remote
      (bind _asfort_game_state_athena (Utilities|Casting|CastToFortGameStateAthena (Game|GetGameState)))
      (if (and (CallFunction|AreTubesEnabled) (SkyTube|IsEnabled))
        (if (not (Variables|/Script/FortniteGame.FortGameStateAthena|GetbPlaylistDataIsLoaded _asfort_game_state_athena))
          (Playlist|BindEventtoOnCurrentPlaylistLoaded _asfort_game_state_athena (delegate Custom|PlaylistLoaded)))
        (Collision|BindEventtoOnComponentBeginOverlap (Variables|Default|GetCapsule) (delegate Custom|PawnPassedThroughConduit))
        (else
          (Audio|Components|Audio|Stop (Variables|Default|GetAmbientAudio))
          (Development|SetHiddeninGame (Variables|Default|GetSM_Wind_Tunnel_Connector) true true))))) ; Everything is cosmetic
)

(event Custom|PlaylistLoaded
  (params (PlaylistName :type Name) (PlaylistContextTags :type Gameplay%20Tag%20Container%20Structure%20%28by%20ref%29)))

(event Custom|PawnPassedThroughConduit
  (params (OverlappedComponent :type Primitive%20Component%20Object%20Reference) (OtherActor :type Actor%20Object%20Reference) (OtherComp :type Primitive%20Component%20Object%20Reference) (OtherBodyIndex :type Integer) (bFromSweep :type Boolean) (SweepResult :type Hit%20Result%20Structure%20%28by%20ref%29))
  (if (and (CallFunction|AreTubesEnabled) (SkyTube|IsEnabled))
    (bind _asfort_pawn (Utilities|Casting|CastToFortPawn OtherActor))
    (Game|Feedback|ClientStartCameraShake :self (Fortnite|GetFortPlayerControllerfromActor _asfort_pawn) :Shake "/Game/Athena/Items/Weapons/Abilities/HookGun/CamerShake_HookGunHitImpact.CamerShake_HookGunHitImpact_C" :Scale 2.0 :PlaySpace "CameraLocal")
    (GameplayCue|ExecuteGameplayCue%28Burst%29Local OtherActor (Variables|Default|GetEnterCue) (Ability|GameplayCue|MakeGameplayCueParameters :NormalizedMagnitude 0.0 :RawMagnitude (Math|Float|MapRangeClamped (Math|Vector|VectorLength (Variables|/Script/Engine.MovementComponent|GetVelocity (Variables|/Script/Engine.Character|GetCharacterMovement _asfort_pawn))) 1.5e+03 2.5e+03 0.0 1.0) :Location "0,0,0" :Normal "0,0,0" :Instigator (Fortnite|GetFortPlayerControllerfromActor _asfort_pawn) :GameplayEffectLevel 1 :AbilityLevel 1 :bReplicateLocationWhenUsingMinimalRepProxy false))))

(fn AreTubesEnabled
  (return (SkyTube|IsEnabled)))

(fn ConstructionScript
  (Utilities|FlowControl|Sequence
    (:then_0
      (Components|StaticMesh|SetStaticMesh (Variables|Default|GetSM_Wind_Tunnel_Connector) (select (Variables|Default|GetNoWindMesh) "/Game/Athena/Environments/Blueprints/SkyTube/FX/SM_Wind_Tunnel_Connector_Blur.SM_Wind_Tunnel_Connector_Blur" "/Game/Athena/Environments/Blueprints/SkyTube/FX/SM_Wind_Tunnel_Connector.SM_Wind_Tunnel_Connector")))
    (:then_1
      (if (Variables|Default|GetRingOnly)
        (Components|DestroyComponent (Variables|Default|GetS_NeoTilted_Conduit_Pole))
        (else
          (Transformation|SetRelativeLocation (Variables|Default|GetS_NeoTilted_Conduit_Base) (Math|Vector|MakeVector 0.0 0.0 (Math|Float|Lerp -1.6e+02 9.6e+02 (Variables|Default|GetBaseOffset))) false false)
          (Transformation|SetRelativeRotation (Variables|Default|GetS_NeoTilted_Conduit_Base) (Math|Rotator|MakeRotator 0.0 0.0 (Variables|Default|GetBase%20Rotation)) false false))))))
