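; Sky tube (slipstream) Blueprint for Athena, Season 18, that fades out visuals and audio when storm phase triggers its destruction.
; OnDestroyDelayStarted linearly fades the spline audio component to silence over the supplied Delay.
; A FadeOverTime timeline then drives SlipstreamsAlpha in the global FortniteMaterialParameters collection, noted as one second and independent of Delay.
; The construction script only calls its parent.
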(event OnDestroyDelayStarted(DestroyOnSafeZonePhase)
  (params (Delay :type Float%20%28single-precision%29))
  ;[1 1 1] FORT-408371
  ; When destroying skytube based on storm phase, fade out visuals/audio over time.
  (bind _returnvalue (Control|GetAudioComponent (Variables|Default|GetFortSplineAudio)))
  (Audio|Components|Audio|FadeOut _returnvalue Delay 0.0 "Linear" :!class /Script/Engine.AudioComponent)
  (Utilities|FlowControl|Sequence
    (:then_1
      (bind (_direction _alpha) (Timeline|FadeOverTime PlayFromStart)
        (:Update
          (Rendering|Material|SetScalarParameterValue "/Game/Packages/Fortress_SharedMaterials/GlobalMaterialParameters/FortniteMaterialParameters.FortniteMaterialParameters" "SlipstreamsAlpha" _alpha :!class /Script/Engine.KismetMaterialLibrary))) ; fade out slipstreams over 1 second
    )))

(fn ConstructionScript
  (|CallParentFunction))
