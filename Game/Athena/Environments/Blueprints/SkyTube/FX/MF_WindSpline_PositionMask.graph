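# Material function for SkyTube wind splines that outputs an opacity mask limited to a Start and End Position along the spline.
# It inverts the V texture coordinate and measures its normalized distance from the midpoint of that range.
# A Fade Distance parameter, scaled by the range width, ramps the mask off beyond both ends.
# Saturate and SmoothStep then clamp and ease the falloff.
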MaterialFunction /Game/Athena/Environments/Blueprints/SkyTube/FX/MF_WindSpline_PositionMask
graph {
 properties {description="Used for Wind Splines, to set the Start and End Positions for opacity fading."}
 nodes [
  n1 = ComponentMask(None=OneMinus(None=TextureCoordinate())) {g=true}
  n2 = ScalarParameter() {parameterName=Start Position}
  n3 = ScalarParameter() {defaultValue=1, parameterName=End Position}
  n4 = Divide(A=Add(A=@n2, B=@n3))
  n5 = Divide(B=Subtract(A=@n3, B=@n2)) {constA=2}
  n6 = Abs(None=Multiply(A=Subtract(A=@n1, B=@n4) {constB=0.5}, B=@n5) {constA=1, constB=2})
  n7 = Multiply(A=@n5, B=ScalarParameter() {defaultValue=0.1, parameterName=Fade Distance})
  n8 = Divide(A=Add(A=OneMinus(None=@n6), B=@n7) {constB=0}, B=@n7) {constB=1}
  FunctionOutput(None=SmoothStep(Value=Saturate(None=@n8)))
 ]
}
