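# Translucent, two-sided unlit material for the SkyTube wind tunnel (slipstream) spline tube, tinted by a time-of-day curve.
# Panned turbulence blends toward a looping band whose direction follows WindTunnelDirection, weighted by WindTunnelOpacityAdd.
# That pattern drives emissive and opacity, shaped by axis-aligned Fresnels, depth, distance and view-angle fades, a spline mask and vertex alpha.
# SlipstreamsAlpha fades it out unless exempt from shutdown; quality and mobile switches simplify the effect.
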Material /Game/Athena/Environments/Blueprints/SkyTube/FX/M_Wind_Tunnel_Tube
graph {
 properties {
  bUseTranslucencyVertexFog=false
  bUsedWithNanite=true
  bUsedWithSplineMeshes=true
  blendMode=BLEND_Translucent
  pixelDepthOffsetMode=PDOM_Legacy
  shadingModel=MSM_Unlit
  twoSided=true
  usedShadingModels=Unlit
 }
 nodes [
  AppendVector(A=TextureCoordinate(), B=Constant())
  MaterialFunctionCall(
   Color_Morning=VectorParameter() {parameterName=MorningTint, defaultValue=<rgba 1.15 0.8884 0.805 0>}.RGB
   Color_Day=VectorParameter() {parameterName=DayTint, defaultValue=<rgba 1.0813 1.2 1.1958 0>}.RGB
   Color_Evening=VectorParameter() {parameterName=EveningTint, defaultValue=<rgba 1.1 0.7845 0.77 1>}.RGB
   Color_Night=VectorParameter() {parameterName=NightTint, defaultValue=<rgba 0.2656 0.85 0.85 0>}.RGB
  ) {materialFunction=&Color_Time_of_Day}
  n1 = CurveAtlasRowParameter(CurveTime=CollectionParameter() {collection=&FortniteMaterialParameters, parameterName=TimeOfDay_0-1}) {
   atlas=&AthenaCurveAtlas01
   curve=/Game/Athena/LinearColorCurves/Curve_TimeOfDay_S23
   defaultValue=321
   group=TimeOfDay
   parameterName=TOD_Curve
  }
  n2 = ScalarParameter() {defaultValue=0.65, parameterName=TimeOfDayColorContribution}
  n3 = Multiply(
   A=LinearInterpolate(A=VectorParameter() {parameterName=BaseColor, defaultValue=<rgba 0.4177 1 0.9 0>}.RGB, B=@n1, Alpha=@n2)
   B=ScalarParameter() {defaultValue=0.02, parameterName=ShadowMult}
  ) {constB=0.1}
  n4 = VertexInterpolator(VS=LinearInterpolate(B=@n1, Alpha=@n2) {constA=1.2})
  n5 = ScalarParameter() {defaultValue=0.5, group=Opacity, parameterName=EdgeAdd}
  n6 = MaterialFunctionCall() {materialFunction=&TangentBasis}
  n7 = MaterialFunctionCall(Axis=@n6.Y V, Exponent=ScalarParameter() {defaultValue=0.8, group=Opacity, parameterName=InnerFresnelExp}) {desc=Inner fade, materialFunction=&AxisAlignedFresnel}
  n8 = OneMinus(None=@n7.Fresnel)
  n9 = ScalarParameter() {defaultValue=1, parameterName=XTile}
  n10 = ScalarParameter() {defaultValue=1, parameterName=YTile}
  n11 = ComponentMask(None=TextureCoordinate() {coordinateIndex=1}) {g=true}
  n12 = ScalarParameter() {defaultValue=1, parameterName=UVSlant}
  n13 = Constant()
  n14 = Add(A=Multiply(A=TextureCoordinate(), B=AppendVector(A=@n9, B=@n10)), B=Multiply(A=Saturate(None=@n11), B=AppendVector(A=@n12, B=@n13)))
  n15 = AppendVector(A=ScalarParameter() {defaultValue=0.15, parameterName=PanSpeedX}, B=ScalarParameter() {defaultValue=-0.3, parameterName=PanSpeedY})
  n16 = Panner(
   Coordinate=Add(A=@n14) {constB=0.5}
   Speed=Multiply(A=@n15, B=ScalarParameter() {defaultValue=-0.9, parameterName=DuplicatedTexture_DirectionMultiply}) {constB=-0.9}
  )
  n17 = Panner(Coordinate=@n14, Speed=@n15)
  n18 = Time() {bOverride_Period=true, period=1}
  n19 = CollectionParameter() {collection=/Game/Athena/Environments/Blueprints/SkyTube/MPC_Gameplay_BPs, parameterName=WindTunnelDirection}
  n20 = LinearInterpolate(A=@n18, B=OneMinus(None=@n18), Alpha=Round(None=@n19))
  n21 = SphereMask(A=ComponentMask(None=TextureCoordinate()) {g=true}, B=@n20, Radius=ScalarParameter() {defaultValue=0.15, parameterName=BlurPanRadius}) {hardnessPercent=0}
  n22 = ScalarParameter() {defaultValue=4, parameterName=XTile2}
  n23 = ScalarParameter() {defaultValue=1, parameterName=YTile2}
  n24 = Panner(
   Coordinate=Multiply(A=TextureCoordinate() {coordinateIndex=1}, B=AppendVector(A=@n22, B=@n23))
   Time=@n20
   Speed=AppendVector(A=ScalarParameter() {defaultValue=-1, parameterName=PanSpeedX2}, B=ScalarParameter() {defaultValue=1, parameterName=PanSpeedY2})
  )
  n25 = Multiply(
   A=@n21
   B=LinearInterpolate(
    A=ScalarParameter() {defaultValue=0.5, parameterName=BlurPanAdd}
    Alpha=TextureSampleParameter2D(UVs=@n24) {parameterName=Texture2, texture=/Game/Effects/Fort_Effects/Textures/Plasma/T_Turbulence_Stretched}.G
   )
  )
  n26 = Multiply(A=@n21) {constB=0.5}
  n27 = ShadingPathSwitch(Default=Multiply(A=@n25, B=@n25), Mobile=@n26) {desc=Migrated from FeatureLevelSwitch}
  n28 = VertexColor()
  n29 = Saturate(None=@n19)
  n30 = OneMinus(None=Abs(None=Sine(None=@n29) {period=2}))
  n31 = CollectionParameter() {collection=&MPC_Gameplay_BPs, parameterName=WindTunnelOpacityAdd}
  n32 = LinearInterpolate(
   A=LinearInterpolate(
    A=TextureSampleParameter2D(UVs=@n16) {parameterName=Texture, texture=&T_Turbulence_Stretched}.G
    B=TextureSampleParameter2D(UVs=@n17) {parameterName=Texture, texture=&T_Turbulence_Stretched}.G
   )
   B=Multiply(A=QualitySwitch(Default=@n27, Low=@n26), B=Multiply(A=@n28.R, B=@n30))
   Alpha=Clamp(None=@n31) {maxDefault=0.64}
  )
  n33 = QualitySwitch(Default=LinearInterpolate(A=Multiply(A=@n5, B=@n8) {constB=0.1}, Alpha=@n32), Low=Add(B=@n32) {constA=0.15})
  MaterialFunctionCall(Emissive=Add(A=VertexInterpolator(VS=@n3).PS, B=Multiply(A=@n4.PS, B=@n33)) {constB=0.2}) {materialFunction=&MF_FX_EmissiveAdjustments}
  n34 = ScalarParameter() {group=Opacity, parameterName=Fade Sensitivity}
  n35 = Multiply(A=@n31) {constB=0.05, desc=.15}
  n36 = Add(A=Multiply(A=@n32, B=ScalarParameter() {defaultValue=0.2, group=Opacity, parameterName=MiddleFadeOpacity}) {constB=0.5}, B=@n35)
  n37 = Add(A=@n35) {constB=0.02}
  n38 = QualitySwitch(
   Default=Constant() {r=2}
   Epic=LinearInterpolate(
    A=ScalarParameter() {defaultValue=3, parameterName=MaxFresnelExp}
    B=ScalarParameter() {defaultValue=1, parameterName=MinFresnelExp}
    Alpha=@n32
   )
  )
  n39 = Multiply(
   A=LinearInterpolate(B=QualitySwitch(Default=@n36, Low=@n37), Alpha=Saturate(None=@n7.Fresnel)) {constA=1, constB=0.2}
   B=ShadingPathSwitch(
    Default=MaterialFunctionCall(Axis=@n6.Y V, Exponent=@n38) {desc=Edge fade, materialFunction=&AxisAlignedFresnel}.Fresnel
    Mobile=Constant() {r=0.3}
   ) {desc=Migrated from FeatureLevelSwitch}
  )
  n40 = ScalarParameter() {defaultValue=150, group=Opacity, parameterName=DepthFadeDist}
  n41 = Multiply(
   A=MaterialFunctionCall() {materialFunction=&/MF_WindSpline_PositionMask}.Result
   B=ShadingPathSwitch(Default=DepthFade(Opacity=@n39, FadeDistance=@n40), Mobile=@n39)
  )
  n42 = CollectionParameter() {collection=&MPC_Gameplay_BPs, parameterName=WindTunnelOpacityAdd}
  n43 = ComponentMask(None=TextureCoordinate()) {g=true}
  n44 = Constant() {r=0.5}
  n45 = ScalarParameter() {defaultValue=0.3, group=Opacity, parameterName=OpacityFadeHardness}
  n46 = QualitySwitch(
   Default=LinearInterpolate(
    B=ScalarParameter() {defaultValue=0.3, group=Opacity, parameterName=CenterOpacity}
    Alpha=SphereMask(A=@n43, B=@n44, Radius=@n44, Hardness=@n45)
   ) {constA=1}
   Low=Constant() {r=0.7}
  )
  n47 = Distance(A=WorldPosition().XYZ, B=CameraPositionWS())
  n48 = ScalarParameter() {defaultValue=12000, parameterName=InnerFaceFadeDistance}
  n49 = LinearInterpolate(
   A=Add(A=Multiply(A=@n42) {constB=0.55}, B=@n46)
   B=Multiply(A=@n46, B=ScalarParameter() {defaultValue=0.35, group=Opacity, parameterName=DistantOpacity})
   Alpha=Saturate(None=Divide(A=@n47, B=@n48))
  ) {constA=1}
  n50 = CollectionParameter() {collection=&FortniteMaterialParameters, parameterName=SlipstreamsAlpha}
  n51 = ScalarParameter() {parameterName=ExemptFromShutdown, sliderMax=1}
  n52 = Multiply(
   A=Multiply(A=Multiply(A=@n41, B=@n49), B=ScalarParameter() {defaultValue=2.5, group=Opacity, parameterName=OpacityMult}) {constB=2}
   B=Saturate(None=Add(A=@n50, B=@n51))
  )
  n53 = Multiply(
   A=@n52
   B=MaterialFunctionCall(
    Value_Morning=ScalarParameter() {defaultValue=1, parameterName="Value_Morning (S)"}
    Value_Day=ScalarParameter() {defaultValue=1, parameterName="Value_Day (S)"}
    Value_Evening=ScalarParameter() {defaultValue=1, parameterName="Value_Evening (S)"}
    Value_Night=ScalarParameter() {defaultValue=1, parameterName="Value_Night (S)"}
   ) {materialFunction=&Scalar_Time_of_Day}.Result
  )
  Multiply(
   A=Power(Base=VertexColor().A, Exp=ScalarParameter() {defaultValue=5, group=Opacity, parameterName=Vertex Alpha Exponent}) {constExponent=5}
   B=Multiply(
    A=MaterialFunctionCall("FadeSensitivity (optional)"=@n34) {materialFunction=&FadeBasedOnViewAngle}.Result
    B=StaticSwitchParameter(True=@n53, False=@n52) {parameterName=UseTODAlphaIntensity?}
   )
  )
 ]
}
