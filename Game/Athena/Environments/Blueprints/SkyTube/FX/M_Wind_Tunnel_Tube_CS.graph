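# Translucent unlit, two-sided spline-mesh material for the SkyTube wind tunnel (slipstream) tube.
# Distorted panning turbulence blends with streaks driven by MPC_Gameplay_BPs wind direction and opacity boost, feeding time-of-day-tinted color and opacity.
# Axis-aligned Fresnel, depth, center and camera-distance fades shape opacity, which global SlipstreamsAlpha scales unless ExemptFromShutdown overrides it.
# Quality and shading-path switches simplify low-end and mobile rendering.
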Material /Game/Athena/Environments/Blueprints/SkyTube/FX/M_Wind_Tunnel_Tube_CS
graph {
 properties {
  bUsedWithSplineMeshes=true
  blendMode=BLEND_Translucent
  pixelDepthOffsetMode=PDOM_Legacy
  shadingModel=MSM_Unlit
  twoSided=true
  usedShadingModels=Unlit
 }
 nodes [
  n1 = VectorParameter() {parameterName=WreckColor, defaultValue=<rgba 0.4177 1 0.9 0>}
  n2 = Multiply(A=Time(), B=ScalarParameter() {defaultValue=0.1, group=Adv Emissive, parameterName=AnimMaskTime})
  n3 = Divide(
   A=MaterialFunctionCall!(f2, f3=ActorPositionWS(), f4) {materialFunction=/Game/Packages/Fortnite_UniqueMaterials_1/VectorComponentOps}.f3 Add
   B=ScalarParameter() {defaultValue=50, group=Adv Emissive, parameterName=Random Distance}
  )
  n4 = TextureSampleParameter2D(UVs=Frac(None=Add(A=@n2, B=@n3))) {
   group=Adv Emissive
   parameterName=AnimMask
   samplerSource=SSM_Wrap_WorldGroupSettings
   texture=/Game/Global/Textures/Default/Flowmaps/T_SwirlFlow_02_D
  }
  n5 = TextureCoordinate()
  n6 = Add() {constA=-1}
  n7 = Multiply(A=Constant2Vector() {g=1}, B=Divide(A=Ceil(None=@n6)) {constB=4})
  n8 = ScalarParameter() {parameterName=X}
  n9 = ScalarParameter() {parameterName=Y}
  n10 = TextureSampleParameter2D(UVs=Panner(Coordinate=Add(A=@n5, B=@n7), Speed=AppendVector(A=@n8, B=@n9))) {
   group=Adv Emissive
   parameterName=EmissiveMask
   samplerSource=SSM_Wrap_WorldGroupSettings
   samplerType=SAMPLERTYPE_LinearColor
   texture=/Game/Environments/Apollo/Sets/Skyfire/Textures/T_Skyfire_PackA
  }
  n11 = DotProduct(A=@n4.RGB, B=@n10.RGB)
  n12 = ScalarParameter() {defaultValue=150, parameterName=DepthFadeDistColor}
  n13 = MaterialFunctionCall(
   Color_Morning=VectorParameter() {parameterName=MorningTint, defaultValue=<rgba 1.15 0.8884 0.805 0>}.RGB
   Color_Day=VectorParameter() {parameterName=DayTint, defaultValue=<rgba 1.0813 1.2 1.1958 0>}.RGB
   Color_Evening=VectorParameter() {parameterName=EveningTint, defaultValue=<rgba 1.1 0.7845 0.77 1>}.RGB
   Color_Night=VectorParameter() {parameterName=NightTint, defaultValue=<rgba 0.2656 0.85 0.85 0>}.RGB
  ) {materialFunction=&Color_Time_of_Day}
  n14 = ScalarParameter() {defaultValue=0.65, parameterName=TimeOfDayColorContribution}
  n15 = LinearInterpolate(
   A=LinearInterpolate(
    A=Multiply(A=@n1.RGB, B=@n11)
    B=VectorParameter() {parameterName=BaseColor, defaultValue=<rgba 0.4177 1 0.9 0>}.RGB
    Alpha=DepthFade(FadeDistance=@n12)
   )
   B=@n13.Result
   Alpha=@n14
  )
  n16 = MaterialFunctionCall() {materialFunction=&TangentBasis}
  n17 = MaterialFunctionCall(Axis=@n16.Y V, Exponent=ScalarParameter() {defaultValue=0.8, parameterName=InnerFresnelExp}) {desc=Inner fade, materialFunction=&AxisAlignedFresnel}
  n18 = Multiply(A=ScalarParameter() {defaultValue=0.5, parameterName=EdgeAdd}, B=OneMinus(None=@n17.Fresnel)) {constB=0.1}
  n19 = ScalarParameter() {defaultValue=1, parameterName=DistortAmount}
  n20 = ScalarParameter() {defaultValue=1, parameterName=XTile}
  n21 = ScalarParameter() {defaultValue=1, parameterName=YTile}
  n22 = ComponentMask(None=TextureCoordinate() {coordinateIndex=1}) {g=true}
  n23 = ScalarParameter() {defaultValue=1, parameterName=UVSlant}
  n24 = Constant()
  n25 = Add(A=Multiply(A=TextureCoordinate(), B=AppendVector(A=@n20, B=@n21)), B=Multiply(A=Saturate(None=@n22), B=AppendVector(A=@n23, B=@n24)))
  n26 = AppendVector(A=ScalarParameter() {defaultValue=0.15, parameterName=PanSpeedX}, B=ScalarParameter() {defaultValue=-0.3, parameterName=PanSpeedY})
  n27 = Panner(Coordinate=@n25, Speed=@n26)
  n28 = TextureSampleParameter2D(UVs=Add(A=Multiply(A=@n10.G, B=@n19), B=@n27)) {parameterName=Texture, texture=/Game/Effects/Fort_Effects/Textures/Plasma/T_Turbulence_Stretched}
  n29 = Add(A=Multiply(A=@n28.G, B=ScalarParameter() {defaultValue=1, parameterName=DistortAmount}), B=@n27)
  n30 = ScalarParameter() {parameterName=BlurSpeedX}
  n31 = ScalarParameter() {defaultValue=0.1, parameterName=BlurSpeedY}
  n32 = TextureSample(UVs=Panner(Coordinate=TextureCoordinate(), Speed=AppendVector(A=@n30, B=@n31))) {texture=/Game/Global/Textures/Default/Masks_and_Gradients/gradient_stasis}
  n33 = ScalarParameter() {defaultValue=0.5, parameterName=BlurPanAdd}
  n34 = TextureCoordinate() {coordinateIndex=2}
  n35 = AppendVector(A=ScalarParameter() {defaultValue=4, parameterName=XTile2}, B=ScalarParameter() {defaultValue=1, parameterName=YTile2})
  n36 = Time() {bOverride_Period=true, period=1}
  n37 = CollectionParameter() {collection=/Game/Athena/Environments/Blueprints/SkyTube/MPC_Gameplay_BPs, parameterName=WindTunnelDirection}
  n38 = LinearInterpolate(A=@n36, B=OneMinus(None=@n36), Alpha=Round(None=@n37))
  n39 = ScalarParameter() {defaultValue=-1, parameterName=PanSpeedX2}
  n40 = ScalarParameter() {defaultValue=1, parameterName=PanSpeedY2}
  n41 = TextureSampleParameter2D(UVs=Panner(Coordinate=Multiply(A=@n34, B=@n35), Time=@n38, Speed=AppendVector(A=@n39, B=@n40))) {parameterName=Texture2, texture=&T_Turbulence_Stretched}
  n42 = Multiply(A=@n33, B=@n41.G)
  n43 = Multiply(A=@n32.G) {constB=0.5}
  n44 = ShadingPathSwitch(Default=Power(Base=Multiply(A=@n32.G, B=@n42)), Mobile=@n43) {desc=Migrated from FeatureLevelSwitch}
  n45 = VertexColor()
  n46 = Saturate(None=@n37)
  n47 = OneMinus(None=Abs(None=Sine(None=@n46) {period=2}))
  n48 = CollectionParameter() {collection=&MPC_Gameplay_BPs, parameterName=WindTunnelOpacityAdd}
  n49 = LinearInterpolate(
   A=LinearInterpolate(
    A=@n28.G
    B=TextureSampleParameter2D(UVs=@n29) {parameterName=Texture, texture=&T_Turbulence_Stretched}.G
   )
   B=Multiply(A=QualitySwitch(Default=@n44, Low=@n43), B=Multiply(A=@n45.R, B=@n47))
   Alpha=Clamp(None=@n48) {maxDefault=0.64}
  )
  n50 = Multiply(
   A=LinearInterpolate(B=@n13.Result, Alpha=@n14) {constA=1.2}
   B=QualitySwitch(Default=LinearInterpolate(A=@n18, Alpha=@n49), Low=Add(B=@n49) {constA=0.15})
  )
  n51 = ScalarParameter() {defaultValue=0.1, parameterName=BandColor}
  Add(
   A=Multiply(A=@n15, B=ScalarParameter() {defaultValue=0.02, parameterName=ShadowMult}) {constB=0.1}
   B=Multiply(A=@n15, B=Multiply(A=@n50, B=@n51))
  ) {constB=0.2}
  LinearInterpolate(A=@n33, Alpha=@n41.G)
  n52 = Multiply(A=@n48) {constB=0.05, desc=.15}
  n53 = Add(A=Multiply(A=@n49, B=ScalarParameter() {defaultValue=0.2, parameterName=MiddleFadeOpacity}) {constB=0.5}, B=@n52)
  n54 = Add(A=@n52) {constB=0.02}
  n55 = Constant() {r=2}
  n56 = QualitySwitch(
   Default=LinearInterpolate(
    A=ScalarParameter() {defaultValue=3, parameterName=MaxFresnelExp}
    B=ScalarParameter() {defaultValue=1, parameterName=MinFresnelExp}
    Alpha=@n49
   )
   Low=@n55
   Medium=@n55
  )
  n57 = Multiply(
   A=LinearInterpolate(B=QualitySwitch(Default=@n53, Low=@n54), Alpha=Saturate(None=@n17.Fresnel)) {constA=1, constB=0.2}
   B=ShadingPathSwitch(
    Default=MaterialFunctionCall(Axis=@n16.Y V, Exponent=@n56) {desc=Edge fade, materialFunction=&AxisAlignedFresnel}.Fresnel
    Mobile=Constant() {r=0.3}
   ) {desc=Migrated from FeatureLevelSwitch}
  )
  n58 = ShadingPathSwitch(Default=DepthFade(Opacity=@n57, FadeDistance=ScalarParameter() {defaultValue=150, parameterName=DepthFadeDist}), Mobile=@n57) {desc=Migrated from FeatureLevelSwitch}
  n59 = CollectionParameter() {collection=&MPC_Gameplay_BPs, parameterName=WindTunnelOpacityAdd}
  n60 = ComponentMask(None=TextureCoordinate()) {g=true}
  n61 = Constant() {r=0.5}
  n62 = ScalarParameter() {defaultValue=0.3, parameterName=OpacityFadeHardness}
  n63 = QualitySwitch(
   Default=LinearInterpolate(B=ScalarParameter() {defaultValue=0.3, parameterName=CenterOpacity}, Alpha=SphereMask(A=@n60, B=@n61, Radius=@n61, Hardness=@n62)) {constA=1}
   Low=Constant() {r=0.7}
  )
  n64 = Distance(A=WorldPosition().XYZ, B=CameraPositionWS())
  n65 = ScalarParameter() {defaultValue=12000, parameterName=InnerFaceFadeDistance}
  n66 = LinearInterpolate(
   A=Add(A=Multiply(A=@n59) {constB=0.55}, B=@n63)
   B=Multiply(A=@n63, B=ScalarParameter() {defaultValue=0.35, parameterName=DistantOpacity})
   Alpha=Saturate(None=Divide(A=@n64, B=@n65))
  ) {constA=1}
  n67 = CollectionParameter() {collection=&FortniteMaterialParameters, parameterName=SlipstreamsAlpha}
  n68 = ScalarParameter() {parameterName=ExemptFromShutdown, sliderMax=1}
  Multiply(
   A=Multiply(A=Multiply(A=@n58, B=@n66), B=ScalarParameter() {defaultValue=2.5, parameterName=OpacityMult}) {constB=2}
   B=Clamp(None=Add(A=@n67, B=@n68))
  )
  Panner(
   Coordinate=Add(A=@n25) {constB=0.5}
   Speed=Multiply(A=@n26, B=ScalarParameter() {defaultValue=-0.9, parameterName=DuplicatedTexture_DirectionMultiply}) {constB=-0.9}
  )
  ScalarParameter() {defaultValue=1, parameterName=TEMPWindTunnelADD}
  ScalarParameter() {parameterName=TEMPWindTunnelDirection}
  SphereMask(A=ComponentMask(None=TextureCoordinate()) {g=true}, B=@n38, Radius=ScalarParameter() {defaultValue=0.15, parameterName=BlurPanRadius}) {hardnessPercent=0}
 ]
}
