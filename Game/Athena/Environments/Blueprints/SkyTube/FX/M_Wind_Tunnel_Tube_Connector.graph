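# Translucent, unlit, two-sided spline-mesh material for the SkyTube wind tunnel tube connector, drawing streaking wind.
# World position offset math stretches it along local X and pushes it along inverted normals by a UV1 gradient.
# Slanted, panning turbulence, tinted by time of day, feeds color and opacity.
# Opacity adds gradient, fresnel, depth and near-camera fades, scaled by global SlipstreamsAlpha; mobile and lower-quality switches simplify noise and fades.
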Material /Game/Athena/Environments/Blueprints/SkyTube/FX/M_Wind_Tunnel_Tube_Connector
graph {
 properties {
  bUseTranslucencyVertexFog=false
  bUsedWithSplineMeshes=true
  blendMode=BLEND_Translucent
  pixelDepthOffsetMode=PDOM_Legacy
  shadingModel=MSM_Unlit
  twoSided=true
  usedShadingModels=Unlit
 }
 nodes [
  n1 = TextureCoordinate()
  n2 = Multiply(A=Constant3Vector() {constant=<rgba 1 0 0 0>}, B=Subtract(A=ComponentMask(None=@n1) {g=true}) {constB=0.5})
  n3 = Multiply(A=VertexNormalWS()) {constB=-1}
  n4 = ScalarParameter() {defaultValue=150, parameterName=WPOSpread}
  n5 = OneMinus(None=ComponentMask(None=TextureCoordinate() {coordinateIndex=1}) {g=true})
  Add(
   A=Multiply(A=Transform(None=@n2) {transformSourceType=TRANSFORMSOURCE_Local}, B=ScalarParameter() {defaultValue=600, parameterName=WPOLengthAdd})
   B=Multiply(A=Multiply(A=@n3, B=@n4), B=Multiply(A=@n5, B=@n5))
  )
  n6 = VectorParameter() {parameterName=BaseColor, defaultValue=<rgba 0.4177 1 0.9 0>}
  n7 = MaterialFunctionCall(
   Color_Morning=VectorParameter() {parameterName=MorningTint, defaultValue=<rgba 1.15 0.8884 0.805 0>}.RGB
   Color_Day=VectorParameter() {parameterName=DayTint, defaultValue=<rgba 1.0812 1.2 1.1958 0>}.RGB
   Color_Evening=VectorParameter() {parameterName=EveningTint, defaultValue=<rgba 1.1 0.7845 0.77 1>}.RGB
   Color_Night=VectorParameter() {parameterName=NightTint, defaultValue=<rgba 0.2656 0.85 0.85 0>}.RGB
  ) {materialFunction=&Color_Time_of_Day}
  n8 = ScalarParameter() {defaultValue=0.65, parameterName=TimeOfDayColorContribution}
  n9 = ScalarParameter() {defaultValue=2, parameterName=XTile}
  n10 = ScalarParameter() {defaultValue=0.15, parameterName=YTile}
  n11 = Power(
   Base=ComponentMask(None=TextureCoordinate() {coordinateIndex=1}) {g=true}
   Exp=ScalarParameter() {defaultValue=3.5, parameterName=UVSlantPower}
  )
  n12 = ScalarParameter() {defaultValue=-0.25, parameterName=UVSlant}
  n13 = Constant()
  n14 = Add(
   A=Multiply(A=TextureCoordinate() {coordinateIndex=1}, B=AppendVector(A=@n9, B=@n10))
   B=Multiply(A=Saturate(None=@n11), B=AppendVector(A=@n12, B=@n13))
  )
  n15 = Constant2Vector() {g=-0.9, r=-1.05}
  n16 = AppendVector(A=ScalarParameter() {defaultValue=-0.2, parameterName=PanSpeedX}, B=ScalarParameter() {defaultValue=-0.5, parameterName=PanSpeedY})
  n17 = TextureSampleParameter2D(UVs=Panner(Coordinate=@n14, Speed=Multiply(A=@n15, B=@n16) {constB=-0.9})) {parameterName=Texture, texture=/Game/Effects/Fort_Effects/Textures/Plasma/T_Turbulence_Stretched}
  n18 = TextureSampleParameter2D(UVs=Panner(Coordinate=@n14, Speed=@n16)) {parameterName=Texture, texture=&T_Turbulence_Stretched}
  n19 = ShadingPathSwitch(Default=QualitySwitch(Default=LinearInterpolate(A=@n17.G, B=@n18.G), Low=@n18.G, High=@n18.G, Medium=@n18.G), Mobile=@n18.G)
  Add(
   A=Multiply(A=LinearInterpolate(A=@n6.RGB, B=@n7.Result, Alpha=@n8), B=ScalarParameter() {defaultValue=0.02, parameterName=ShadowMult}) {constB=0.1}
   B=Multiply(A=LinearInterpolate(B=@n7.Result, Alpha=@n8) {constA=1.4}, B=@n19)
  ) {constB=0.2}
  n20 = ComponentMask(None=TextureCoordinate() {coordinateIndex=1}) {g=true}
  n21 = Add(A=@n19) {constB=0.1}
  n22 = ScalarParameter() {defaultValue=2, parameterName=MaxFresnelExp}
  n23 = ScalarParameter() {defaultValue=0.6, parameterName=MinFresnelExp}
  n24 = Constant() {r=0.6}
  n25 = QualitySwitch(
   Default=MaterialFunctionCall(
    Axis=MaterialFunctionCall() {materialFunction=&TangentBasis}.Y V
    Exponent=LinearInterpolate(A=@n22, B=@n23, Alpha=@n19)
   ) {desc=Edge fade, materialFunction=&AxisAlignedFresnel}.Fresnel
   Low=@n24
   High=@n24
   Medium=@n24
  )
  n26 = Multiply(
   A=DepthFade(Opacity=Multiply(A=@n21, B=@n25), FadeDistance=ScalarParameter() {defaultValue=100, parameterName=DepthFadeDist})
   B=MaterialFunctionCall(Distance=ScalarParameter() {defaultValue=700, parameterName=NearCamFade}) {materialFunction=&Near_Camera_Fade}.Result
  )
  n27 = CollectionParameter() {collection=&FortniteMaterialParameters, parameterName=SlipstreamsAlpha}
  Multiply(A=Multiply(A=Multiply(A=@n20, B=@n20), B=ShadingPathSwitch(Default=@n26, Mobile=@n21)) {constA=2}, B=@n27)
  Power(Base=@n5) {constExponent=2.3}
  Power(Base=@n20) {constExponent=1.5}
 ]
}
