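# Additive, unlit, two-sided material for the blur effect on SkyTube wind tunnel connectors, usable on spline meshes.
# Radial coordinates from UV1 pan a water-blur mask, which drives an opacity value scaled by the global SlipstreamsAlpha parameter.
# Color fakes lighting from time-of-day tint, sun/moon reflection highlights and shadow tint, with a flat constant on mobile.
# That lighting is multiplied by a fan-blade gradient and the panning mask.
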Material /Game/Athena/Environments/Blueprints/SkyTube/FX/M_Wind_Tunnel_Tube_ConnectorBlur
graph {
 properties {
  bUsedWithSplineMeshes=true
  blendMode=BLEND_Additive
  pixelDepthOffsetMode=PDOM_Legacy
  shadingModel=MSM_Unlit
  twoSided=true
  usedShadingModels=Unlit
 }
 nodes [
  n1 = MaterialFunctionCall(Vector or UVs=Subtract(A=TextureCoordinate() {coordinateIndex=1, uTiling=2, vTiling=2})) {materialFunction=&VectorToRadialValue}
  n2 = AppendVector(A=ScalarParameter() {defaultValue=3, parameterName=X}, B=ScalarParameter() {defaultValue=7, parameterName=Y})
  n3 = Time()
  n4 = ScalarParameter() {defaultValue=3, parameterName=PanSpeed}
  n5 = TextureSample(UVs=Panner(Coordinate=Multiply(A=@n1.Radial Coordinates, B=@n2), Time=Multiply(A=@n3, B=@n4)) {speedX=-1, speedY=-0.5}) {
   samplerType=SAMPLERTYPE_LinearGrayscale
   texture=/Game/Characters/Player/Male/Medium/Heads/M_MED_AUX_Alien_Head_01/Textures/T_waterBlur__MASK
  }
  n6 = CollectionParameter() {collection=&FortniteMaterialParameters, parameterName=SlipstreamsAlpha}
  Multiply(A=LinearInterpolate(A=ScalarParameter() {defaultValue=0.2, parameterName=MinOpacity}, Alpha=@n5.G) {constB=1.2}, B=@n6)
  n7 = MaterialFunctionCall() {materialFunction=&Color_Time_of_Day}
  n8 = CollectionParameter() {
   collection=&FortniteMaterialParameters
   parameterName=SunAndMoonModelDirectionalVector
  }
  n9 = DotProduct(A=ComponentMask(None=@n8) {b=true, g=true, r=true}, B=ReflectionVectorWS() {bNormalizeCustomWorldNormal=true})
  n10 = ScalarParameter() {defaultValue=25, parameterName=Spec Highlight Power}
  n11 = Multiply(A=Multiply(A=Add(A=@n7.Result) {constB=0.25}) {constB=0.8}, B=Saturate(None=Power(Base=@n9, Exp=@n10) {constExponent=20}))
  n12 = Saturate(None=Power(Base=@n9))
  n13 = Add(A=MaterialFunctionCall() {materialFunction=&Color_Time_of_Day}.Result) {constB=0.25}
  n14 = Add(
   A=Multiply(A=Multiply(A=@n12, B=@n13)) {constB=0.5}
   B=Multiply(A=@n13, B=VectorParameter() {parameterName=ShadowTint, defaultValue=<rgba 0.1953 0.2215 0.3 0>}.RGB) {constB=0.05}
  )
  Multiply(
   A=ShadingPathSwitch(Default=Add(A=@n11, B=@n14), Mobile=Constant() {r=0.7}) {desc=Migrated from FeatureLevelSwitch}
   B=Multiply(A=TextureSample() {texture=&/T_Fan_Blade_Gradient}.RGB, B=Add(A=@n5.RGB) {constB=0.7})
  )
 ]
}
