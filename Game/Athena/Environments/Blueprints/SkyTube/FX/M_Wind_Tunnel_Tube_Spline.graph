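# Translucent, unlit, two-sided material for the SkyTube wind tunnel tube, drawn on spline meshes widened by SplineThicken with a flat color.
# Opacity combines two power curves on a U-coordinate gradient, with exponents blended by camera depth fade so the profile changes with distance.
# Two blended turbulence samples, panning with opposite vertical motion over skewed UVs, break up that opacity.
# The global SlipstreamsAlpha collection parameter scales the result.
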Material /Game/Athena/Environments/Blueprints/SkyTube/FX/M_Wind_Tunnel_Tube_Spline
graph {
 properties {
  bUsedWithInstancedStaticMeshes=true
  bUsedWithSplineMeshes=true
  blendMode=BLEND_Translucent
  ditheredLODTransition=true
  pixelDepthOffsetMode=PDOM_Legacy
  shadingModel=MSM_Unlit
  twoSided=true
  usedShadingModels=Unlit
 }
 nodes [
  MaterialFunctionCall(Emissive=VectorParameter() {parameterName=Color, defaultValue=<rgba 0.5052 0.5052 0.5052 1>}.RGB) {materialFunction=&MF_FX_EmissiveAdjustments}
  n1 = ScalarParameter() {defaultValue=943.26, parameterName=Width}
  MaterialFunctionCall(WidthBase=@n1, WidthTip=@n1) {materialFunction=&SplineThicken}
  n2 = ComponentMask(None=TextureCoordinate()) {r=true}
  n3 = Multiply(A=Multiply(A=@n2, B=OneMinus(None=@n2))) {constB=2}
  n4 = MaterialFunctionCall(
   Fade Length=ScalarParameter() {defaultValue=20000, parameterName=FadeCamDistance}
   Fade Offset=ScalarParameter() {defaultValue=60000, parameterName=FadeCamOffset}
  ) {materialFunction=&CameraDepthFade}
  n5 = Power(
   Base=@n3
   Exp=LinearInterpolate(A=ScalarParameter() {defaultValue=0.5, parameterName=Fresnelnear}, B=ScalarParameter() {parameterName=Fresnel}, Alpha=@n4.Result)
  )
  n6 = Power(
   Base=OneMinus(None=@n3)
   Exp=LinearInterpolate(
    A=ScalarParameter() {defaultValue=5, parameterName=Fresnel2near}
    B=ScalarParameter() {defaultValue=2.2769, parameterName=Fresnel2}
    Alpha=@n4.Result
   )
  )
  n7 = ScalarParameter() {defaultValue=0.1733, parameterName=MiddleOpacityMin}
  n8 = TextureCoordinate()
  n9 = TextureCoordinate()
  n10 = Multiply(A=Saturate(None=ComponentMask(None=@n9) {g=true}), B=Constant2Vector() {r=1})
  n11 = Add(A=@n8, B=Multiply(A=@n10) {constB=-1})
  n12 = Add(A=@n8, B=@n10)
  n13 = LinearInterpolate(
   A=TextureSampleParameter2D(UVs=Panner(Coordinate=@n11) {speedX=0.2625, speedY=0.45}) {parameterName=Texture, texture=/Game/Global/Textures/Default/Caustics/T_Turbulence_Seamless}.G
   B=TextureSampleParameter2D(UVs=Panner(Coordinate=@n12) {speedX=0.25, speedY=-0.5}) {parameterName=Texture, texture=&T_Turbulence_Seamless}.G
  )
  n14 = CollectionParameter() {collection=&FortniteMaterialParameters, parameterName=SlipstreamsAlpha}
  Multiply(A=Multiply(A=Multiply(A=@n5, B=@n6), B=LinearInterpolate(A=@n7, Alpha=@n13)), B=@n14)
 ]
}
