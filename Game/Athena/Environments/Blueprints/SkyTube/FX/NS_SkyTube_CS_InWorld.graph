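# Niagara system for the in-world SkyTube effect, scattering glowing purple, red and blue particles along a user-supplied spline.
# Three emitters layer large rotating sprites, small speed-stretched sprites, and magenta sprites with light renderers and curl noise.
# Each places particles randomly between User.MinPosition and User.Progress, drifting slightly along the spline with age, plus a random offset.
# The large-sprite layer uses distance culling and distance-remapped size and alpha.
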NiagaraSystem /Game/Athena/Environments/Blueprints/SkyTube/FX/NS_SkyTube_CS_InWorld
dependencies {
 usedDataInterfaces [
  &NiagaraDataInterfaceMeshRendererInfo
  &NiagaraDataInterfaceColorCurve
  &NiagaraDataInterfaceCurve
  &NiagaraDataInterfaceArrayFloat
 ]
 usedDynamicInputs []
 usedModules [&SystemState]
}
emitters {
 FN_Sprite {
  emitterUpdateScript [
   &EmitterState(
    Enable Distance Culling=true
    MaxDistance=50000
    Min Distance Response=::SleepAndLetParticlesFinish
    MinDistance=2500
    Module.bUseMaxDistance=true
    Module.bUseMinDistance=true
    Scalability Mode=::NewEnumerator1
   )
   &SpawnRate(SpawnRate=6, Use Spawn Probability=false)
  ]
  particleSpawnScript [
   &InitializeParticle(
    Color=&RandomRangeLinearColor(Minimum=<rgba 0.1394 0.0225 0.1912 1>, Maximum=<rgba 1 1 1 1>)
    Lifetime Max=3
    Lifetime Min=2
    Lifetime Mode=::NewEnumerator1
    Sprite Rotation Mode=::NewEnumerator2
    Sprite Size Mode=::NewEnumerator1
    Uniform Sprite Size Max=1600
    Uniform Sprite Size Min=1500
    UsePositionOffset=false
    Write Material Random=true
   ) {module=InitializeParticle001}
   /SetVariables(Particles.SpriteRotationRate=&RandomRangeFloat(Minimum=-1, Maximum=1, Evaluation Type=::NewEnumerator1)) {isSetParameters=true, module=SetVariables_#1}
   /SetVariables() {isSetParameters=true, module=SetVariables_#2}
   /SetVariables() {isSetParameters=true, module=SetVariables_#3}
   /SetVariables(Particles.MaxPosition:NiagaraFloat=$User.Progress) {isSetParameters=true, module=SetVariables_#4}
   /SetVariables(Particles.MinPosition:NiagaraFloat=$User.MinPosition) {isSetParameters=true, module=SetVariables_#5}
  ]
  particleUpdateScript [
   &ParticleState()
   &DynamicMaterialParameters(
    Index 0 Param 1=&RandomRangeFloat(Minimum=0, Maximum=1)
    Index 0 Param 2=&RandomRangeFloat(Minimum=0, Maximum=1)
    Index 0 Param 3=1
    Module.Param0WriteEnabled=true
    Module.Param1WriteEnabled=true
    Module.Param2WriteEnabled=true
    Module.Param3WriteEnabled=true
   ) {module=DynamicMaterialParameters001}
   &ScaleColor(
    Scale Alpha=&FloatFromCurve(
     FloatCurve=DataInterface() {propertyValues {Curve {keys [
         {interpMode=RCIM_Cubic, tangentMode=RCTM_User, time=0, value=0}
         <at 0.2 1>
         {interpMode=RCIM_Cubic, tangentMode=RCTM_User, time=0.8, value=1}
         {interpMode=RCIM_Cubic, tangentMode=RCTM_User, time=1, value=0}
        ]}}}
     Scale Curve=&Multiply_Float(
      A=0.25
      B=&FN_LOD_Distance_Remap(Start Distance=500, End Distance=25000, Minimum Output=0.5, Maximum Output=1.1)
     )
    )
    Scale RGB=&Multiply_VectorByFloat(
     Vector=&MakeVectorFromLinearColorRGB(Color=&MakeLinearColorFromVectorAndFloat(
       "Vector (RGB)"=&MakeVectorFromLinearColorRGB(Color=&ColorFromCurve(ColorCurve=DataInterface() {propertyValues {
           AlphaCurve {keys [<at 0 1 RCIM_Cubic>,<at 0.9977 1 RCIM_Cubic>]}
           BlueCurve {keys [<at 0.1965 0.3372 RCIM_Cubic>,<at 0.3626 0>,<at 0.626 1>,<at 0.8083 1.9736>]}
           GreenCurve {keys [<at 0.1965 0 RCIM_Cubic>,<at 0.3626 0.026>,<at 0.626 0>,<at 0.8083 0>]}
           RedCurve {keys [<at 0.1965 0.0686 RCIM_Cubic>,<at 0.3626 1>,<at 0.626 0.5952>,<at 0.8083 3>]}
          }}))
       "Float (Alpha)"=&FloatFromCurve(FloatCurve=DataInterface() {propertyValues {Curve {keys [
            {interpMode=RCIM_Cubic, tangentMode=RCTM_User, time=0, value=1}
            {interpMode=RCIM_Cubic, tangentMode=RCTM_User, time=0.5, value=1}
            {interpMode=RCIM_Cubic, tangentMode=RCTM_User, time=1, value=1}
           ]}}})
      ))
     Float=15
    )
    ScaleA=true
    ScaleRGB=true
   )
   &SpriteRotationRate(Delta Time:NiagaraFloat=$Particles.NormalizedAge, Rotation Rate=$Particles.SpriteRotationRate)
   &self:Spline_Position(
    Position Along Spline:NiagaraFloat=&Add_Float(
     A=&RandomRangeFloat(Minimum=$Particles.MinPosition, Maximum=$Particles.MaxPosition)
     B=&Divide_Float(
      A=$Particles.NormalizedAge
      B=&SelectFloatFromArray(
       Array Sampling Mode=::NewEnumerator0
       Direct Array Index=?
       Float Selection Array=DataInterface() {propertyValues {FloatData [100,-100]}}
       Random Seed=0
      )
     )
    )
    Position Offset:Vector3f=&RandomRangeVector(Minimum=<50 50 50>, Maximum=<-50 -50 -50>)
    Spline:NiagaraDataInterfaceSpline=$User.Spline
   ) {module=Spline_Position}
   &SolveForcesAndVelocity(Clamp Velocity=false, Limit Acceleration=false)
   &ScaleSpriteSize(
    Uniform Curve Scale=&Multiply_Float(
     A=1
     B=&FN_LOD_Distance_Remap(Start Distance=5000, End Distance=15000, Minimum Output=0.5, Maximum Output=1.25)
    )
    Uniform Curve Sprite Scale=DataInterface() {propertyValues {Curve {keys [{tangentMode=RCTM_User, time=0, value=0}, <at 0.1 0>, <at 0.3 1>, {tangentMode=RCTM_User, time=1, value=1}]}}}
   )
  ]
  rendererClasses [&NiagaraSpriteRendererProperties]
  renderers [{rendererClass=&NiagaraSpriteRendererProperties}]
 }
 FN_Sprite001 {
  emitterUpdateScript [
   &EmitterState(Life Cycle Mode=::NewEnumerator1, UseLoopDelay=false)
   &SpawnRate(SpawnRate=100, Use Spawn Probability=false)
  ]
  particleSpawnScript [
   &InitializeParticle(
    Color=&RandomRangeLinearColor(Minimum=<rgba 1 1 1 1>, Maximum=<rgba 1 1 1 1>)
    Lifetime Max=1
    Lifetime Min=0.5
    Lifetime Mode=::NewEnumerator1
    Sprite Size Mode=::NewEnumerator1
    Uniform Sprite Size Max=100
    Uniform Sprite Size Min=10
    UsePositionOffset=false
    Write Material Random=true
   ) {module=InitializeParticle001}
   /SetVariables(Particles.SpriteRotationRate=&RandomRangeFloat(Minimum=-1, Maximum=1, Evaluation Type=::NewEnumerator1)) {isSetParameters=true, module=SetVariables_#1}
   /SetVariables(Particles.AmplitudeSmoothed=1) {isSetParameters=true, module=SetVariables_#2}
   /SetVariables() {isSetParameters=true, module=SetVariables_#3}
   /SetVariables(Particles.MaxPosition:NiagaraFloat=$User.Progress) {isSetParameters=true, module=SetVariables_#4}
   /SetVariables(Particles.MinPosition:NiagaraFloat=$User.MinPosition) {isSetParameters=true, module=SetVariables_#5}
   /SetVariables(Particles.ParticleSizeFloat:NiagaraFloat=/Game/Effects/Niagara/DynamicInputs/Math/FN_Fit(
     Clamp:NiagaraBool=false
     Value:NiagaraFloat=&MakeFloatFromVector2d(Vector 2D=$Particles.SpriteSize)
     SourceMin:NiagaraFloat=10
     SourceMax:NiagaraFloat=100
     DestMin:NiagaraFloat=0.5
     DestMax:NiagaraFloat=0.05
    )) {isSetParameters=true}
   /SetVariables(Particles.GlowSize=&Multiply_Vector2D(A=$Particles.SpriteSize, B=<0.25 0.25>)) {isSetParameters=true}
  ]
  particleUpdateScript [
   &ParticleState()
   &DynamicMaterialParameters(
    Index 0 Param 1=&RandomRangeFloat(Minimum=0, Maximum=1)
    Index 0 Param 2=&RandomRangeFloat(Minimum=0, Maximum=1)
    Index 0 Param 3=1
    Module.Param0WriteEnabled=true
    Module.Param1WriteEnabled=true
    Module.Param2WriteEnabled=true
    Module.Param3WriteEnabled=true
   ) {module=DynamicMaterialParameters001}
   &ScaleColor(
    Scale Alpha=&FloatFromCurve(
     FloatCurve=DataInterface() {propertyValues {Curve {keys [
         {interpMode=RCIM_Cubic, tangentMode=RCTM_User, time=0, value=0}
         {interpMode=RCIM_Cubic, tangentMode=RCTM_User, time=0.5, value=1}
         {interpMode=RCIM_Cubic, tangentMode=RCTM_User, time=1, value=0}
        ]}}}
     Scale Curve=$Particles.ParticleSizeFloat
    )
    Scale RGB=&Multiply_VectorByFloat(
     Vector=&MakeVectorFromLinearColorRGB(Color=&MakeLinearColorFromVectorAndFloat(
       "Vector (RGB)"=&MakeVectorFromLinearColorRGB(Color=&ColorFromCurve(ColorCurve=DataInterface() {propertyValues {
           AlphaCurve {keys [<at 0 1 RCIM_Cubic>,<at 0.9977 1 RCIM_Cubic>]}
           BlueCurve {keys [<at 0.1297 0.3372 RCIM_Cubic>,<at 0.3829 0.8309>,<at 0.5328 0.3372>,<at 0.8083 1.9736>]}
           GreenCurve {keys [<at 0.1297 0 RCIM_Cubic>,<at 0.3829 0>,<at 0.5328 0.0169>,<at 0.8083 0>]}
           RedCurve {keys [<at 0.1297 0.0686 RCIM_Cubic>,<at 0.3829 1>,<at 0.5328 0.0058>,<at 0.8083 3>]}
          }}))
       "Float (Alpha)"=&FloatFromCurve(FloatCurve=DataInterface() {propertyValues {Curve {keys [
            {interpMode=RCIM_Cubic, tangentMode=RCTM_User, time=0, value=1}
            {interpMode=RCIM_Cubic, tangentMode=RCTM_User, time=0.5, value=1}
            {interpMode=RCIM_Cubic, tangentMode=RCTM_User, time=1, value=1}
           ]}}})
      ))
     Float=25
    )
    ScaleA=true
    ScaleRGB=true
   )
   &SpriteRotationRate(Delta Time:NiagaraFloat=$Particles.NormalizedAge, Rotation Rate=$Particles.SpriteRotationRate)
   &self:Spline_Position(
    Position Along Spline:NiagaraFloat=&Add_Float(
     A=&RandomRangeFloat(Minimum=$Particles.MinPosition, Maximum=$Particles.MaxPosition)
     B=&Divide_Float(
      A=$Particles.NormalizedAge
      B=&SelectFloatFromArray(
       Array Sampling Mode=::NewEnumerator0
       Direct Array Index=?
       Float Selection Array=DataInterface() {propertyValues {FloatData [200,-200]}}
       Random Seed=0
      )
     )
    )
    Position Offset:Vector3f=&RandomRangeVector(Minimum=<400 400 400>, Maximum=<-400 -400 -400>)
    Spline:NiagaraDataInterfaceSpline=$User.Spline
   ) {module=Spline_Position}
   &CurlNoiseForce(Noise Frequency=5, Noise Quality / Cost=::NewEnumerator0, Noise Strength=0, Pan Noise=true)
   &SolveForcesAndVelocity(Clamp Velocity=false, Limit Acceleration=false)
   /Niagara/Modules/Update/Camera/MaintainInCameraParticleScale()
   &CalculateAccurateVelocity(Set Intrinsic Variables=false)
   &ScaleSpriteSizeBySpeed(Max Scale Factor=<1 5>, Source Velocity:Vector3f=$Particles.CalculateAccurateVelocity.Velocity, Velocity Threshold=100)
   &ScaleSpriteSizeBySpeed(
    Initial Sprite Size:Vector2f=$Particles.Initial.GlowSize
    Max Scale Factor=<1 15>
    Min Scale Factor=<1 0.25>
    Source Velocity:Vector3f=$Particles.CalculateAccurateVelocity.Velocity
    Velocity Threshold=100
   ) {enabled=false, module=ScaleSpriteSizeBySpeed001}
  ]
  rendererClasses [&NiagaraSpriteRendererProperties]
  renderers [{rendererClass=&NiagaraSpriteRendererProperties}]
 }
 Light {
  emitterUpdateScript [
   &EmitterState(Life Cycle Mode=::NewEnumerator1, UseLoopDelay=false)
   &SpawnRate(SpawnRate=3, Use Spawn Probability=false)
  ]
  particleSpawnScript [
   &InitializeParticle(
    Color=&RandomRangeLinearColor(Minimum=<rgba 1 1 1 1>, Maximum=<rgba 1 1 1 1>)
    Lifetime Max=1
    Lifetime Min=0.5
    Lifetime Mode=::NewEnumerator1
    Sprite Size Mode=::NewEnumerator1
    Uniform Sprite Size Max=25
    Uniform Sprite Size Min=15
    UsePositionOffset=false
    Write Material Random=true
   ) {module=InitializeParticle001}
   /SetVariables(Particles.SpriteRotationRate=&RandomRangeFloat(Minimum=-1, Maximum=1, Evaluation Type=::NewEnumerator1)) {isSetParameters=true, module=SetVariables_#1}
   /SetVariables(Particles.AmplitudeSmoothed=1) {isSetParameters=true, module=SetVariables_#2}
   /SetVariables() {isSetParameters=true, module=SetVariables_#3}
   /SetVariables(Particles.MaxPosition:NiagaraFloat=$User.Progress) {isSetParameters=true, module=SetVariables_#4}
   /SetVariables(Particles.MinPosition:NiagaraFloat=$User.MinPosition) {isSetParameters=true, module=SetVariables_#5}
  ]
  particleUpdateScript [
   &ParticleState()
   &DynamicMaterialParameters(
    Index 0 Param 1=&RandomRangeFloat(Minimum=0, Maximum=1)
    Index 0 Param 2=&RandomRangeFloat(Minimum=0, Maximum=1)
    Index 0 Param 3=1
    Module.Param0WriteEnabled=true
    Module.Param1WriteEnabled=true
    Module.Param2WriteEnabled=true
    Module.Param3WriteEnabled=true
   ) {module=DynamicMaterialParameters001}
   &ScaleColor(
    Scale Alpha=&FloatFromCurve(
     FloatCurve=DataInterface() {propertyValues {Curve {keys [
         {interpMode=RCIM_Cubic, tangentMode=RCTM_User, time=0, value=0}
         {interpMode=RCIM_Cubic, tangentMode=RCTM_User, time=0.5, value=1}
         {interpMode=RCIM_Cubic, tangentMode=RCTM_User, time=1, value=0}
        ]}}}
     Scale Curve=0.5
    )
    Scale RGB=<100 0 100>
    ScaleA=true
    ScaleRGB=true
   )
   &SpriteRotationRate(Delta Time:NiagaraFloat=$Particles.NormalizedAge, Rotation Rate=$Particles.SpriteRotationRate)
   &self:Spline_Position(
    Position Along Spline:NiagaraFloat=&Add_Float(
     A=&RandomRangeFloat(Minimum=$Particles.MinPosition, Maximum=$Particles.MaxPosition)
     B=&Divide_Float(
      A=$Particles.NormalizedAge
      B=&SelectFloatFromArray(
       Array Sampling Mode=::NewEnumerator0
       Direct Array Index=?
       Float Selection Array=DataInterface() {propertyValues {FloatData [200,-200]}}
       Random Seed=0
      )
     )
    )
    Position Offset:Vector3f=&RandomRangeVector(Minimum=<100 100 100>, Maximum=<-100 -100 -100>)
    Spline:NiagaraDataInterfaceSpline=$User.Spline
   ) {module=Spline_Position}
   &CurlNoiseForce(Noise Frequency=5, Noise Quality / Cost=::NewEnumerator0, Noise Strength=5000, Pan Noise=true, Pan Noise Field=<0 1 2>)
   &SolveForcesAndVelocity(Clamp Velocity=false, Limit Acceleration=false)
   &ScaleSpriteSizeBySpeed(Max Scale Factor=<1 1>, Velocity Threshold=100) {enabled=false}
   &MaintainInCameraParticleScale()
  ]
  rendererClasses [&NiagaraSpriteRendererProperties, &NiagaraLightRendererProperties]
  renderers [
   {rendererClass=&NiagaraSpriteRendererProperties}
   {rendererClass=&NiagaraLightRendererProperties, rendererIndex=1}
  ]
 }
}
summary {
 systemName=NS_SkyTube_CS_InWorld
 emitters [
  {emitterName=FN_Sprite, rendererClasses [&NiagaraSpriteRendererProperties]}
  {emitterName=FN_Sprite001, rendererClasses [&NiagaraSpriteRendererProperties]}
  {emitterName=Light, rendererClasses [&NiagaraSpriteRendererProperties, &NiagaraLightRendererProperties]}
 ]
 userVariables [
  User.MinPosition {type=NiagaraFloat, defaultValue=0.3, description=""}
  User.New Spline {
   type=NiagaraDataInterfaceSpline
   description=""
   defaultValue {
    struct=NiagaraExt_VariableValue_DataInterface
    value {
     dataInterface=&self:NiagaraDataInterfaceSpline_0
     dataInterfaceClass=/Script/Niagara.NiagaraDataInterfaceSpline
    }
   }
  }
  User.Progress {type=NiagaraFloat, defaultValue=0.4, description=""}
  User.RingLocation {type=NiagaraFloat, defaultValue=0, description=""}
  User.Spline {
   type=NiagaraDataInterfaceSpline
   description=""
   defaultValue {
    struct=NiagaraExt_VariableValue_DataInterface
    value {
     dataInterface=&self:NiagaraDataInterfaceSpline_1
     dataInterfaceClass=&NiagaraDataInterfaceSpline
    }
   }
  }
 ]
}
