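# Niagara system for SkyTube speed-line effects, using one GPU sprite emitter that continuously spawns pink-magenta particles with distance-based culling and spawn scaling.
# Particles spawn on a large torus, with User.RingLocation driving its U distribution, plus a sphere-surface location offset.
# Randomized vortex velocity in either direction and strong curl noise move them, while speed scales sprite size.
# An alpha curve fades them in quickly and out over life.
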NiagaraSystem /Game/Athena/Environments/Blueprints/SkyTube/FX/NS_SkyTube_CS_SpeedLines
dependencies {
 usedDataInterfaces [&NiagaraDataInterfaceCurve, &NiagaraDataInterfaceColorCurve, &NiagaraDataInterfaceVector2DCurve]
 usedDynamicInputs []
 usedModules [&SystemState]
}
emitters {FN_Sprite_SpeedLines001 {
  simTarget=GPUComputeSim
  emitterUpdateScript [
   &SpawnBurst_Instantaneous(Spawn Count=5, Use Spawn Probability=false) {enabled=false}
   &EmitterState(
    Enable Distance Culling=true
    MaxDistance=15000
    Module.bApplySpawnCountScaleByDistance=false
    Module.bUseMaxDistance=true
    Module.bUseMinDistance=false
    Scalability Mode=::NewEnumerator1
    Scale Spawn Count=true
   )
   &SpawnRate(SpawnRate=1500, Use Spawn Probability=false)
  ]
  particleSpawnScript [
   &InitializeParticle(
    Color Maximum=<rgba 5 0.2865 3.977 1>
    Color Minimum=<rgba 4.5171 0.452 5 0.25>
    Color Mode=::NewEnumerator2
    Sprite Size Mode=::NewEnumerator3
    Uniform Sprite Size=1500
    UsePositionOffset=false
    Write Material Random=true
   )
   &SystemLocation(MaskSpawnGroup=false) {enabled=false}
   &TorusLocation(
    Align Torus With Axis=true
    Handle Radius=0
    Large Radius=5250
    MaskSpawnGroup=false
    Module.Override Randomness=false
    Torus Axis=<0 0 0>
    U Distribution:NiagaraFloat=$User.RingLocation
   )
   &CylinderLocation(
    Cylinder Height=500
    Cylinder Radius=5250
    Hemicircle Internal Cap=false
    Hemicircle X=true
    Hemicircle Y=true
    MaskSpawnGroup=false
    Offset Mode:ENiagara_OffsetMode=::NewEnumerator0
    Override Seed=false
    Surface Only=true
    Use Endcaps In Surface Only Mode=false
    UseOwnerScale=true
   ) {enabled=false}
   &SphereLocation(
    MaskSpawnGroup=false
    Override Seed=false
    Sphere Radius=400
    Surface Expansion Mode=::NewEnumerator2
    Surface Only=true
    UseNonUniformScale=false
    UseOffset=false
    UseSphereOrientation=false
   )
   /SetVariables(Particles.Initial.SpriteSize:Vector2f=&Multiply_Vector2(
     A=<1 1>
     B=&Vector2DFromFloat(Value=&Multiply_Float(
       A=1
       B=&FloatFromCurve(
        FloatCurve=DataInterface() {propertyValues {Curve {keys [
            {arriveTangent=0.4214, interpMode=RCIM_Cubic, leaveTangent=0.4225, tangentMode=RCTM_User, time=25, value=0}
            {arriveTangent=12.969, interpMode=RCIM_Cubic, leaveTangent=12.969, tangentMode=RCTM_User, time=31.9201, value=25.8333}
            {arriveTangent=7.1685, interpMode=RCIM_Cubic, leaveTangent=7.1685, tangentMode=RCTM_User, time=41.1282, value=209.1666}
            {arriveTangent=16.7131, interpMode=RCIM_Cubic, leaveTangent=16.7131, tangentMode=RCTM_User, time=70, value=500}
            <at 90 1500>
           ]}}}
        CurveIndex=150
       )
      ))
    )) {isSetParameters=true}
  ]
  particleUpdateScript [
   &ParticleState()
   &Color(
    Color=&ScaleLinearColorByCurve(
     InitialValue=<rgba 1 1 1 0.075>
     ScaleCurve=DataInterface() {propertyValues {
       AlphaCurve {keys [<at 0.0384 1>,<at 0.3631 1>]}
       BlueCurve {keys [<at 0.0281 1>]}
       GreenCurve {keys [<at 0.0281 1>]}
       RedCurve {keys [<at 0.0281 1>]}
      }}
     CurveIndex=$Particles.NormalizedAge
    )
    Scale Alpha=&FloatFromCurve(FloatCurve=DataInterface() {propertyValues {Curve {keys [<at 0 0 RCIM_Cubic>,<at 0.5 1 RCIM_Cubic>,<at 1 0 RCIM_Cubic>]}}})
   ) {enabled=false}
   &ScaleColor(
    Scale Alpha=&FloatFromCurve(FloatCurve=DataInterface() {propertyValues {Curve {keys [
         {interpMode=RCIM_Cubic, tangentMode=RCTM_User, time=0, value=0}
         {interpMode=RCIM_Cubic, tangentMode=RCTM_User, time=0.1, value=1}
         {interpMode=RCIM_Cubic, tangentMode=RCTM_User, time=1.0011, value=0.002}
        ]}}})
    ScaleA=true
    ScaleRGB=true
   )
   /Niagara/Modules/Update/Velocity/VortexVelocity(
    Module.Use Influence Falloff=false
    Velocity Amount:NiagaraFloat=&Lerp_Float(A=-2500, B=2500, Alpha=&RandomRangeFloat(Minimum=0, Maximum=1))
   )
   &VortexForce(Origin Pull Amount=1500, UseOriginOffset=false, Vortex Force Amount=1500000) {enabled=false}
   &CurlNoiseForce(Noise Frequency=1, Noise Strength=25000, Pan Noise=false)
   &SolveForcesAndVelocity(Clamp Velocity=false, Limit Acceleration=false)
   &ScaleSpriteSizeBySpeed(Max Scale Factor=<0.01 0.1>, Min Scale Factor=<0 0>, Velocity Threshold=2500)
   &ScaleSpriteSize(Scale Factor=&Vector2DFromCurve(
     Vector2Curve=DataInterface() {propertyValues {
       XCurve {keys [<at -0.0069 0.5 RCIM_Cubic>,<at 0.4918 1 RCIM_Cubic>,<at 1 0.5 RCIM_Cubic>]}
       YCurve {keys [<at 0 0.0245 RCIM_Cubic>, {arriveTangent=-0.0245, interpMode=RCIM_Cubic, leaveTangent=-0.0245, time=0.4975, value=1}, <at 1 0 RCIM_Cubic>]}
      }}
     ScaleCurve=<1 1>
    )) {enabled=false}
   &CameraOffset(Camera Offset Amount=10)
  ]
  rendererClasses [&NiagaraSpriteRendererProperties]
  renderers [{rendererClass=&NiagaraSpriteRendererProperties}]
 }}
summary {
 systemName=NS_SkyTube_CS_SpeedLines
 emitters [{emitterName=FN_Sprite_SpeedLines001, simTarget=GPUComputeSim, rendererClasses [&NiagaraSpriteRendererProperties]}]
 userVariables [User.RingLocation {type=NiagaraFloat, defaultValue=0, description=""}]
}
