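; Gameplay ability that lets a player damage buildings they hit, apparently while traveling through a SkyTube.
; On activation, the server binds the avatar pawn's capsule hit event and unbinds it when the ability ends.
; Each hit on a damageable BuildingActor without the Skytube.Safe tag applies GE_SkytubeDamageWallsLOTS from the player's ability system to the building's.
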(event Ability|EventActivateAbility
  (bind _asfort_player_pawn (Utilities|Casting|CastToFortPlayerPawn (Ability|GetAvatarActorfromActorInfo)))
  (Variables|Default|SetPlayerPawn _asfort_player_pawn)
  (if (Networking|IsServer)
    (Collision|BindEventtoOnComponentHit (Variables|/Script/Engine.Character|GetCapsuleComponent (Variables|Default|GetPlayerPawn)) (EventDispatchers|CreateEvent :!event Custom|PlayerCapsuleHit))))

(event Ability|EventOnEndAbility
  (params (bWasCancelled :type Boolean))
  (bind _playerpawn (Variables|Default|GetPlayerPawn))
  (|CallParentFunction bWasCancelled)
  (if (Networking|IsServer)
    (Utilities|IsValid _playerpawn
      (:Is%20Valid
        (Collision|UnbindEventfromOnComponentHit (Variables|/Script/Engine.Character|GetCapsuleComponent _playerpawn) (EventDispatchers|CreateEvent :!event Custom|PlayerCapsuleHit))))))

(event Custom|PlayerCapsuleHit
  (params (HitComponent :type Primitive%20Component%20Object%20Reference) (OtherActor :type Actor%20Object%20Reference) (OtherComp :type Primitive%20Component%20Object%20Reference) (NormalImpulse :type Vector) (Hit :type Hit%20Result%20Structure%20%28by%20ref%29))
  (bind _asbuilding_actor (Utilities|Casting|CastToBuildingActor OtherActor))
  (Utilities|IsValid (Variables|Default|GetPlayerPawn)
    (:Is%20Valid
      (if (Variables|/Script/Engine.Actor|GetbCanBeDamaged _asbuilding_actor)
        (if (not (Actor|ActorHasTag _asbuilding_actor "Skytube.Safe"))
          (GameplayEffects|ApplyGameplayEffectToTarget (Ability|GetAbilitySystemComponent (Variables|Default|GetPlayerPawn)) "/Game/Athena/Environments/Blueprints/SkyTube/GE_SkytubeDamageWallsLOTS.GE_SkytubeDamageWallsLOTS_C" (Ability|GetAbilitySystemComponent _asbuilding_actor) 0.0 (GameplayEffects|MakeEffectContext (Ability|GetAbilitySystemComponent (Variables|Default|GetPlayerPawn)))))))))
