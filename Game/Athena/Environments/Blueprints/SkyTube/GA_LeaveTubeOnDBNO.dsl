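; Gameplay ability that pulls a player out of a SkyTube when they become down but not out (DBNO).
; On activation it caches the avatar as a FortPlayerPawn and binds to that pawn's DBNO state change event.
; When DBNO begins, it exits the current SkyTube, notifies the BP_Athena_SkyTube actor and removes the looping SkyTube gameplay effect.
; The ability never ends itself or unbinds the delegate, and a debug print remains.
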(event Ability|EventActivateAbility
  (bind _asfort_player_pawn (Utilities|Casting|CastToFortPlayerPawn (Ability|GetAvatarActorfromActorInfo)))
  (bind _output_get (Variables|Default|SetPlayerPawn _asfort_player_pawn))
  (EventDispatchers|BindEventtoOnDBNOStateChanged _output_get (delegate Custom|DownButNotOut)))

(event Custom|DownButNotOut
  (params (FortPawn :type Fort%20Pawn%20Object%20Reference) (bInIsDBNO :type Boolean))
  (bind _returnvalue (SkyTube|GetCurrentSkyTube (Variables|Default|GetPlayerPawn)))
  (bind _playerpawn (Variables|Default|GetPlayerPawn))
  (if bInIsDBNO
    (Development|PrintString "Down but not out")
    (SkyTube|ExitSkyTube _playerpawn _returnvalue :!class /Script/FortniteGame.FortPlayerPawn)
    (bind _asbp_athena_sky_tube (Utilities|Casting|CastToBP_Athena_SkyTube _returnvalue)
      (:then
        (Class|BPAthenaSkyTubeBase|ActorExitSkytube _asbp_athena_sky_tube _playerpawn))
      (:CastFailed
        _))
    (GameplayEffects|RemoveActiveGameplayEffectbySourceEffect (Ability|GetAbilitySystemComponent _playerpawn) "/Game/Athena/Environments/Blueprints/SkyTube/GE_Athena_SkyTube_Loop.GE_Athena_SkyTube_Loop_C" 0 -1)))

(event Ability|EventOnEndAbility
  (params (bWasCancelled :type Boolean))
  (|CallParentFunction bWasCancelled))
